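# Control Rig for the female medium PocketScrunchie cosmetic body, adding procedural corrective deformation to its clothing and joints.
# A shared DeformRig function feeds a sequence covering upper arms, hoodie, skirt, phone, tie, collar and laces.
# Weighted parent constraints and leg-aimed constraints position garment bones.
# Corrective offsets are weighted by spherical pose readers, remapped distances and rotations, and interpolated IsKneeSliding, with Skydiving_Alpha switching arm correctives.
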ControlRigBlueprint /BRCosmetics/Characters/Player/Female/Medium/Bodies/F_MED_PocketScrunchie/Meshes/F_MED_PocketScrunchie_CtrlRig
graphs [
 RigVMFunctionLibrary {links [], nodes []}
 RigVMFunction_Sequence_1_ContainedGraph {nodes [
   n1 = RigVMFunctionEntryNode()
   RigVMFunction_Sequence = &RigVMFunction_Sequence(ExecuteContext=@n1.ExecuteContext)
   n2 = &RigVMFunction_Sequence(ExecuteContext=@RigVMFunction_Sequence.B)
   n3 = &RigVMFunction_Sequence(ExecuteContext=@n2.B)
   n4 = &RigVMFunction_Sequence(ExecuteContext=@n3.B)
   n5 = &RigVMFunction_Sequence(ExecuteContext=@n4.B)
   n6 = &RigVMFunction_Sequence(ExecuteContext=@n5.B)
   RigVMFunctionReturnNode(A=@RigVMFunction_Sequence.A, B=@n2.A, C=@n3.A, D=@n4.A, E=@n5.A, F=@n6.A, G=@n6.B)
  ]}
 RigVMModel {nodes [
   n1 = &RigUnit_GetRelativeTransformForItem(Child=(Type=Bone,Name="clavicle_l"), Parent=(Type=Bone,Name="clavicle_l"), bChildInitial=true, bParentInitial=false)
   &RigUnit_ModifyTransforms(
    ExecutePin=()
    ItemToModify=((Item=(Type=Bone,Name="deform_upperarm_fr_l"),Transform=(Rotation=(X=0,Y=0,Z=0,W=1),Translation=(X=0,Y=15,Z=0),Scale3D=(X=1,Y=1,Z=1))))
    WeightMinimum=0
    Weight=&RigVMFunction_MathDoubleRemap(
     SourceMaximum=29
     SourceMinimum=5
     TargetMaximum=1
     TargetMinimum=0
     Value=&RigVMFunction_MathQuaternionToEuler(RotationOrder=ZYX, Value=@n1.RelativeTransform.Rotation).Result.Y
    ).Result
   )
   n2 = &RigUnit_GetRelativeTransformForItem(Child=(Type=Bone,Name="clavicle_r"), Parent=(Type=Bone,Name="clavicle_r"), bChildInitial=true, bParentInitial=False)
   > &RigUnit_ModifyTransforms(
    ItemToModify=((Item=(Type=Bone,Name="deform_upperarm_fr_r"),Transform=(Rotation=(X=0,Y=0,Z=0,W=1),Translation=(X=-20,Y=-10,Z=0),Scale3D=(X=1,Y=1,Z=1))))
    WeightMinimum=0
    Weight=&RigVMFunction_MathDoubleRemap(
     SourceMaximum=29
     SourceMinimum=5
     TargetMaximum=1
     TargetMinimum=0
     Value=&RigVMFunction_MathQuaternionToEuler(RotationOrder=ZYX, Value=@n2.RelativeTransform.Rotation).Result.Y
    ).Result
   )
   &RigUnit_ParentConstraint(
    AdvancedSettings=(RotationOrderForFilter=XZY)
    Child=(Type=Bone,Name="deform_skirt_chain_f_l")
    ExecutePin=()
    Parents="((Item=(Type=Bone,Name=\"dyn_skirt_b_l\"),Weight=0.004),(Item=(Type=Bone,Name=\"dyn_skirt_m_1_l\"),Weight=0.0313),(Item=(Type=Bone,Name=\"dyn_skirt_f_1_l\"),Weight=0.078),(Item=(Type=Bone,Name=\"dyn_skirt_f_l\"),Weight=0.1777),(Item=(Type=Bone,Name=\"dyn_skirt_m_l\" ...(truncated)"
    Weight=1
   )
   > &RigUnit_ParentConstraint(
    AdvancedSettings=(RotationOrderForFilter=XZY)
    Child=(Type=Bone,Name="deform_skirt_chain_f_r")
    Parents="((Item=(Type=Bone,Name=\"dyn_skirt_m_1_l\"),Weight=0.002),(Item=(Type=Bone,Name=\"spine_02\"),Weight=0.0021),(Item=(Type=Bone,Name=\"dyn_skirt_m_l\"),Weight=0.0049),(Item=(Type=Bone,Name=\"pelvis\"),Weight=0.0207),(Item=(Type=Bone,Name=\"dyn_skirt_f_r1\"),Weight=0.0 ...(truncated)"
    Weight=1
   )
   RigUnit_BeginExecution = &RigUnit_BeginExecution()
   FN_DeformRig_FNC_1 = RigVMFunctionReferenceNode(
    DeformRig_ Curve="(deformRig_l_upperarm_up_alpha_72_#1=False,deformRig_r_upperarm_up_alpha_73_#2=False,deformRig_l_upperarm_fr_alpha_74_#3=False,deformRig_r_upperarm_fr_alpha_75_CB51B8 ...(truncated)"
    DeformRig_Bool="(upperarm_1_#4=true,elbow_3_#5=True,wrist_7_#6=true,groin_8_#7=True,glute_17_#8=True,knee_9_40C408324CA261 ...(truncated)"
    DeformRig_Debug=(upperarm_debug_18_#4=False,wrist_debug_20_#6=False,lat_debug_26_#9=False,groinGlute_debug_27_#7=False)
    DeformRig_Value="(upperarm_up_18_#10=(Rotation=(X=0,Y=0,Z=0,W=1),Translation=(X=0,Y=0,Z=8),Scale3D=(X=0.7,Y=1,Z=1)),upperarm_fwd_fr_19_#11=(Rotation=(X=0,Y=0,Z=0,W=1),Translation=(X=-5,Y=15,Z=0),Scale3D=(X=0.5,Y=1,Z ...(truncated)"
    ExecuteContext=@RigUnit_BeginExecution.ExecutePin
   )
   > RigVMFunction_Sequence = RigVMAggregateNode()
   n3 = &RigVMFunction_ControlFlowBranch(ExecuteContext=@RigVMFunction_Sequence.A, Condition=RigVMVariableNode() {variable=Skydiving_Alpha}.Value)
   n4 = &RigUnit_SphericalPoseReader(
    ActiveRegionScaleFactors=(PositiveWidth=2.5,NegativeWidth=3,PositiveHeight=3,NegativeHeight=3)
    ActiveRegionSize=0.1
    Debug=(bDrawDebug=false,DebugScale=25,DebugSegments=20,DebugThickness=0.25)
    DriverAxis=(X=1,Y=0,Z=0)
    DriverItem=(Type=Bone,Name="upperarm_r")
    FalloffRegionScaleFactors=(PositiveWidth=1,NegativeWidth=1,PositiveHeight=1,NegativeHeight=1)
    FalloffSize=0.7
    OptionalParentItem=(Type=Bone)
    RotationOffset=(X=0,Y=0,Z=0)
    ExecutePin=@n3.True
   )
   > &RigUnit_ModifyTransforms(
    ItemToModify="((Item=(Type=Bone,Name=\"deform_upperarm_up_r\"),Transform=(Rotation=(X=0,Y=0,Z=0,W=1),Translation=(X=0,Y=0,Z=6),Scale3D=(X=1,Y=1,Z=1))),(Item=(Type=Bone,Name=\"upperarm_twist_01_r\"),Transform=(Rotation=(X=0,Y=0,Z=0,W=1),Translation=(X=0,Y=0,Z=0),Scale3D=(X=1 ...(truncated)"
    Weight=@n4.OutputParam
   )
   > n5 = &RigUnit_SphericalPoseReader(
    ActiveRegionScaleFactors=(PositiveWidth=2.5,NegativeWidth=3,PositiveHeight=3,NegativeHeight=3)
    ActiveRegionSize=0.1
    Debug=(bDrawDebug=false,DebugScale=25,DebugSegments=20,DebugThickness=0.25)
    DriverAxis=(X=1,Y=0,Z=0)
    DriverItem=(Type=Bone,Name="upperarm_l")
    FalloffRegionScaleFactors=(PositiveWidth=1,NegativeWidth=1,PositiveHeight=1,NegativeHeight=1)
    FalloffSize=0.7
    OptionalParentItem=(Type=Bone)
    RotationOffset=(X=0,Y=0,Z=0)
   )
   > &RigUnit_ModifyTransforms(
    ItemToModify="((Item=(Type=Bone,Name=\"deform_upperarm_up_l\"),Transform=(Rotation=(X=0,Y=0,Z=0,W=1),Translation=(X=0,Y=0,Z=-6),Scale3D=(X=1,Y=1,Z=1))),(Item=(Type=Bone,Name=\"upperarm_twist_01_l\"),Transform=(Rotation=(X=0,Y=0,Z=0,W=1),Translation=(X=0,Y=0,Z=0),Scale3D=(X= ...(truncated)"
    Weight=@n5.OutputParam
   )
   n6 = &RigUnit_SphericalPoseReader(
    ActiveRegionScaleFactors=(PositiveWidth=2.5,NegativeWidth=3,PositiveHeight=3,NegativeHeight=3)
    ActiveRegionSize=0.1
    Debug=(bDrawDebug=false,DebugScale=25,DebugSegments=20,DebugThickness=0.25)
    DriverAxis=(X=1,Y=0,Z=0)
    DriverItem=(Type=Bone,Name="upperarm_r")
    FalloffRegionScaleFactors=(PositiveWidth=1,NegativeWidth=1,PositiveHeight=1,NegativeHeight=1)
    FalloffSize=0.7
    OptionalParentItem=(Type=Bone)
    RotationOffset=(X=0,Y=0,Z=0)
    ExecutePin=@n3.False
   )
   > &RigUnit_ModifyTransforms(
    ItemToModify="((Item=(Type=Bone,Name=\"deform_upperarm_bk_r\"),Transform=(Rotation=(X=0,Y=0,Z=0,W=1),Translation=(X=-2,Y=-3,Z=15),Scale3D=(X=1,Y=1,Z=1))),(Item=(Type=Bone,Name=\"upperarm_twist_02_r\"),Transform=(Rotation=(X=0,Y=0,Z=0,W=1),Translation=(X=0,Y=0,Z=0),Scale3D=( ...(truncated)"
    Weight=@n6.OutputParam
   )
   > n7 = &RigUnit_SphericalPoseReader(
    ActiveRegionScaleFactors=(PositiveWidth=2.5,NegativeWidth=3,PositiveHeight=3,NegativeHeight=3)
    ActiveRegionSize=0.1
    Debug=(bDrawDebug=false,DebugScale=25,DebugSegments=20,DebugThickness=0.25)
    DriverAxis=(X=1,Y=0,Z=0)
    DriverItem=(Type=Bone,Name="upperarm_l")
    FalloffRegionScaleFactors=(PositiveWidth=1,NegativeWidth=1,PositiveHeight=1,NegativeHeight=1)
    FalloffSize=0.7
    OptionalParentItem=(Type=Bone)
    RotationOffset=(X=0,Y=0,Z=0)
   )
   > n8 = &RigUnit_ModifyTransforms(
    ItemToModify="((Item=(Type=Bone,Name=\"deform_upperarm_bk_l\"),Transform=(Rotation=(X=0,Y=0,Z=0,W=1),Translation=(X=2,Y=3,Z=-12),Scale3D=(X=1,Y=1,Z=1))),(Item=(Type=Bone,Name=\"upperarm_twist_02_l\"),Transform=(Rotation=(X=0,Y=0,Z=0,W=1),Translation=(X=0,Y=0,Z=0),Scale3D=(X ...(truncated)"
    Weight=@n7.OutputParam
   )
   RigVMFunction_ControlFlowBranch = &RigVMFunction_ControlFlowBranch(Condition=false, ExecuteContext=@n8.ExecutePin)
   &RigUnit_ParentConstraint(
    AdvancedSettings=(InterpolationType=Average,RotationOrderForFilter=XZY)
    Child=(Type=Bone,Name="upperarm_twist_01_r")
    Parents=((Item=(Type=Bone,Name="upperarm_r"),Weight=1))
    Weight=1
    ExecutePin=@RigVMFunction_ControlFlowBranch.True
   )
   > &RigUnit_ParentConstraint(
    AdvancedSettings=(InterpolationType=Average,RotationOrderForFilter=XZY)
    Child=(Type=Bone,Name="upperarm_twist_01_l")
    Parents=((Item=(Type=Bone,Name="upperarm_l"),Weight=1))
    Weight=1
   )
   > n9 = &RigUnit_SphericalPoseReader(
    ActiveRegionScaleFactors=(PositiveWidth=2.5,NegativeWidth=3,PositiveHeight=3,NegativeHeight=3)
    ActiveRegionSize=0.1
    Debug=(bDrawDebug=false,DebugScale=25,DebugSegments=20,DebugThickness=0.25)
    DriverAxis=(X=1,Y=0,Z=0)
    DriverItem=(Type=Bone,Name="upperarm_r")
    FalloffRegionScaleFactors=(PositiveWidth=1,NegativeWidth=1,PositiveHeight=1,NegativeHeight=1)
    FalloffSize=0.7
    OptionalParentItem=(Type=Bone)
    RotationOffset=(X=0,Y=0,Z=0)
   )
   > &RigUnit_ModifyTransforms(
    ItemToModify=((Item=(Type=Bone,Name="upperarm_twist_01_r"),Transform=(Rotation=(X=-0.3018,Y=-0.0134,Z=0.0095,W=0.9532),Translation=(X=0.4564,Y=0.7731,Z=-1.6851),Scale3D=(X=1,Y=1,Z=1))))
    Weight=@n9.OutputParam
   )
   > n10 = &RigUnit_SphericalPoseReader(
    ActiveRegionScaleFactors=(PositiveWidth=2.5,NegativeWidth=3,PositiveHeight=3,NegativeHeight=3)
    ActiveRegionSize=0.1
    Debug=(bDrawDebug=false,DebugScale=25,DebugSegments=20,DebugThickness=0.25)
    DriverAxis=(X=1,Y=0,Z=0)
    DriverItem=(Type=Bone,Name="upperarm_l")
    FalloffRegionScaleFactors=(PositiveWidth=1,NegativeWidth=1,PositiveHeight=1,NegativeHeight=1)
    FalloffSize=0.7
    OptionalParentItem=(Type=Bone)
    RotationOffset=(X=0,Y=0,Z=0)
   )
   > &RigUnit_ModifyTransforms(
    ItemToModify=((Item=(Type=Bone,Name="upperarm_twist_01_l"),Transform=(Rotation=(X=-0.4848,Y=0,Z=0,W=0.8746),Translation=(X=0,Y=-1.2337,Z=1.1722),Scale3D=(X=1,Y=1,Z=1))))
    Weight=@n10.OutputParam
   )
   &RigUnit_ParentConstraint(
    AdvancedSettings=(RotationOrderForFilter=XZY)
    Child=(Type=Bone,Name="deform_hoodie_l")
    Parents=((Item=(Type=Bone,Name="spine_01"),Weight=0.0199),(Item=(Type=Bone,Name="spine_03"),Weight=0.0456),(Item=(Type=Bone,Name="spine_04"),Weight=0.097),(Item=(Type=Bone,Name="spine_02"),Weight=0.3237),(Item=(Type=Bone,Name="pelvis"),Weight=0.5137))
    Weight=1
    ExecutePin=@RigVMFunction_Sequence.B
   )
   > &RigUnit_ParentConstraint(
    AdvancedSettings=(RotationOrderForFilter=XZY)
    Child=(Type=Bone,Name="deform_hoodie_r")
    Parents="((Item=(Type=Bone,Name=\"dyn_skirt_f_1_l\"),Weight=0.0024),(Item=(Type=Bone,Name=\"spine_03\"),Weight=0.0024),(Item=(Type=Bone,Name=\"spine_01\"),Weight=0.0044),(Item=(Type=Bone,Name=\"deform_groin_r\"),Weight=0.0048),(Item=(Type=Bone,Name=\"spine_02\"),Weight=0.018 ...(truncated)"
    Weight=1
   )
   &RigUnit_ParentConstraint(
    AdvancedSettings=(RotationOrderForFilter=XZY)
    Child=(Type=Bone,Name="deform_skirt_b_l")
    Parents="((Item=(Type=Bone,Name=\"spine_03\"),Weight=0.0013),(Item=(Type=Bone,Name=\"spine_01\"),Weight=0.0027),(Item=(Type=Bone,Name=\"thigh_l\"),Weight=0.0107),(Item=(Type=Bone,Name=\"spine_02\"),Weight=0.0113),(Item=(Type=Bone,Name=\"deform_groin_l\"),Weight=0.029),(Item= ...(truncated)"
    Weight=1
    ExecutePin=@RigVMFunction_Sequence.C
   )
   > &RigUnit_ParentConstraint(
    AdvancedSettings=(RotationOrderForFilter=XZY)
    Child=(Type=Bone,Name="deform_skirt_bk_l")
    Parents="((Item=(Type=Bone,Name=\"spine_01\"),Weight=0.0016),(Item=(Type=Bone,Name=\"deform_glute_r\"),Weight=0.0039),(Item=(Type=Bone,Name=\"spine_02\"),Weight=0.0064),(Item=(Type=Bone,Name=\"thigh_twist_01_l\"),Weight=0.0689),(Item=(Type=Bone,Name=\"deform_glute_l\"),Weigh ...(truncated)"
    Weight=1
   )
   > &RigUnit_ParentConstraint(
    AdvancedSettings=(RotationOrderForFilter=XZY)
    Child=(Type=Bone,Name="deform_skirt_f_l")
    Parents="((Item=(Type=Bone,Name=\"thigh_twist_01_r\"),Weight=0.0033),(Item=(Type=Bone,Name=\"spine_03\"),Weight=0.0039),(Item=(Type=Bone,Name=\"spine_01\"),Weight=0.0075),(Item=(Type=Bone,Name=\"spine_02\"),Weight=0.031),(Item=(Type=Bone,Name=\"thigh_twist_01_l\"),Weight=0.0 ...(truncated)"
    Weight=1
   )
   > &RigUnit_ParentConstraint(
    AdvancedSettings=(RotationOrderForFilter=XZY)
    Child=(Type=Bone,Name="deform_skirt_m_l")
    Parents="((Item=(Type=Bone,Name=\"thigh_l\"),Weight=0.0032),(Item=(Type=Bone,Name=\"spine_03\"),Weight=0.0042),(Item=(Type=Bone,Name=\"spine_01\"),Weight=0.008),(Item=(Type=Bone,Name=\"deform_glute_l\"),Weight=0.0093),(Item=(Type=Bone,Name=\"spine_02\"),Weight=0.032),(Item=( ...(truncated)"
    Weight=1
   )
   > &RigUnit_ParentConstraint(
    AdvancedSettings=(RotationOrderForFilter=XZY)
    Child=(Type=Bone,Name="deform_skirt_m_l")
    Parents=((Item=(Type=Bone,Name="deform_groin_l"),Weight=1))
    Weight=0.2
   )
   > &RigUnit_ParentConstraint(
    AdvancedSettings=(RotationOrderForFilter=XZY)
    Child=(Type=Bone,Name="deform_skirt_m_r")
    Parents="((Item=(Type=Bone,Name=\"thigh_r\"),Weight=0.0021),(Item=(Type=Bone,Name=\"spine_03\"),Weight=0.0041),(Item=(Type=Bone,Name=\"deform_glute_r\"),Weight=0.0076),(Item=(Type=Bone,Name=\"spine_01\"),Weight=0.0078),(Item=(Type=Bone,Name=\"spine_02\"),Weight=0.0307),(Item ...(truncated)"
    Weight=1
   )
   > &RigUnit_ParentConstraint(
    AdvancedSettings=(RotationOrderForFilter=XZY)
    Child=(Type=Bone,Name="deform_skirt_bk_r")
    Parents="((Item=(Type=Bone,Name=\"spine_01\"),Weight=0.001),(Item=(Type=Bone,Name=\"thigh_r\"),Weight=0.0024),(Item=(Type=Bone,Name=\"thigh_twist_01_l\"),Weight=0.0026),(Item=(Type=Bone,Name=\"spine_02\"),Weight=0.0043),(Item=(Type=Bone,Name=\"deform_glute_l\"),Weight=0.0085 ...(truncated)"
    Weight=1
   )
   > &RigUnit_ParentConstraint(
    AdvancedSettings=(RotationOrderForFilter=XZY)
    Child=(Type=Bone,Name="deform_skirt_f_r")
    Parents="((Item=(Type=Bone,Name=\"dyn_skirt_f_1_l\"),Weight=0.0024),(Item=(Type=Bone,Name=\"spine_03\"),Weight=0.0024),(Item=(Type=Bone,Name=\"spine_01\"),Weight=0.0044),(Item=(Type=Bone,Name=\"deform_groin_r\"),Weight=0.0048),(Item=(Type=Bone,Name=\"spine_02\"),Weight=0.018 ...(truncated)"
    Weight=1
   )
   > &RigUnit_ParentConstraint(
    AdvancedSettings=(RotationOrderForFilter=XZY)
    Child=(Type=Bone,Name="deform_skirt_b_r")
    Parents="((Item=(Type=Bone,Name=\"spine_01\"),Weight=0.0017),(Item=(Type=Bone,Name=\"spine_02\"),Weight=0.0077),(Item=(Type=Bone,Name=\"thigh_r\"),Weight=0.0099),(Item=(Type=Bone,Name=\"deform_groin_r\"),Weight=0.0166),(Item=(Type=Bone,Name=\"deform_glute_r\"),Weight=0.1177) ...(truncated)"
    Weight=1
   )
   > n11 = &RigUnit_SphericalPoseReader(
    ActiveRegionScaleFactors=(PositiveWidth=3.5,NegativeWidth=0.5,PositiveHeight=3,NegativeHeight=3)
    ActiveRegionSize=0.1
    Debug=(bDrawDebug=false,DebugScale=25,DebugSegments=20,DebugThickness=0.25)
    DriverAxis=(X=-1,Y=0,Z=0)
    DriverItem=(Type=Bone,Name="thigh_l")
    FalloffRegionScaleFactors=(PositiveWidth=1,NegativeWidth=1,PositiveHeight=1,NegativeHeight=1)
    FalloffSize=0.45
    OptionalParentItem=(Type=Bone)
    RotationOffset=(X=90,Y=0,Z=-30)
   )
   > n12 = &RigUnit_SphericalPoseReader(
    ActiveRegionScaleFactors=(PositiveWidth=3.5,NegativeWidth=0.5,PositiveHeight=3,NegativeHeight=3)
    ActiveRegionSize=0.1
    Debug=(bDrawDebug=false,DebugScale=25,DebugSegments=20,DebugThickness=0.25)
    DriverAxis=(X=1,Y=0,Z=0)
    DriverItem=(Type=Bone,Name="thigh_r")
    FalloffRegionScaleFactors=(PositiveWidth=1,NegativeWidth=1,PositiveHeight=1,NegativeHeight=1)
    FalloffSize=0.45
    FlipHeightScaling=false
    FlipWidthScaling=true
    OptionalParentItem=(Type=Bone)
    RotationOffset=(X=-90,Y=0,Z=0)
   )
   > &RigUnit_AimConstraintLocalSpaceOffset(
    AdvancedSettings=(DebugSettings=(bEnabled=true,Scale=10,WorldOffset=(Rotation=(X=0,Y=0,Z=0,W=1),Translation=(X=0,Y=0,Z=0),Scale3D=(X=1,Y=1,Z=1))),RotationOrderForFilter=XZY)
    AimAxis=(X=-1,Y=0,Z=0)
    Child=(Type=Bone,Name="deform_skirt_f_l")
    Filter=(bX=True,bY=True,bZ=True)
    Parents=((Item=(Type=Bone,Name="calf_l"),Weight=1))
    UpAxis=(X=0,Y=1,Z=0)
    Weight=1
    WorldUp=(Target=(X=0,Y=0,Z=1),Kind=Location,Space=(Type=Bone,Name="calf_l"))
   )
   > &RigUnit_AimConstraintLocalSpaceOffset(
    AdvancedSettings=(DebugSettings=(bEnabled=true,Scale=10,WorldOffset=(Rotation=(X=0,Y=0,Z=0,W=1),Translation=(X=0,Y=0,Z=0),Scale3D=(X=1,Y=1,Z=1))),RotationOrderForFilter=XZY)
    Child=(Type=Bone,Name="deform_skirt_f_r")
    Filter=(bX=True,bY=True,bZ=True)
    Parents=((Item=(Type=Bone,Name="calf_r"),Weight=1))
    UpAxis=(X=0,Y=-1,Z=0)
    Weight=1
    WorldUp=(Target=(X=0,Y=0,Z=1),Kind=Location,Space=(Type=Bone,Name="calf_r"))
   )
   > &RigUnit_ModifyTransforms(
    ItemToModify=((Item=(Type=Bone,Name="deform_skirt_f_l"),Transform=(Rotation=(X=0.1209,Y=0.2094,Z=0.2219,W=0.9446),Translation=(X=-5,Y=5,Z=0),Scale3D=(X=1,Y=1,Z=1))))
    WeightMinimum=0
    Weight=@n11.OutputParam
   )
   > &RigUnit_ModifyTransforms(
    ItemToModify=((Item=(Type=Bone,Name="deform_skirt_f_r"),Transform=(Rotation=(X=-0.0076,Y=0.1735,Z=0.043,W=0.9839),Translation=(X=0,Y=-5,Z=0),Scale3D=(X=1,Y=1,Z=1))))
    WeightMinimum=0
    Weight=@n12.OutputParam
   )
   > &RigUnit_AimConstraintLocalSpaceOffset(
    AdvancedSettings=(DebugSettings=(bEnabled=true,Scale=10,WorldOffset=(Rotation=(X=0,Y=0,Z=0,W=1),Translation=(X=0,Y=0,Z=0),Scale3D=(X=1,Y=1,Z=1))),RotationOrderForFilter=XZY)
    AimAxis=(X=-1,Y=0,Z=0)
    Child=(Type=Bone,Name="deform_skirt_m_l")
    Filter=(bX=True,bY=True,bZ=True)
    Parents=((Item=(Type=Bone,Name="calf_l"),Weight=1))
    UpAxis=(X=0,Y=1,Z=0)
    Weight=1
    WorldUp=(Target=(X=0,Y=0,Z=1),Kind=Location,Space=(Type=Bone,Name="deform_skirt_m_l"))
    bMaintainOffset=true
   )
   > &RigUnit_AimConstraintLocalSpaceOffset(
    AdvancedSettings=(DebugSettings=(bEnabled=true,Scale=10,WorldOffset=(Rotation=(X=0,Y=0,Z=0,W=1),Translation=(X=0,Y=0,Z=0),Scale3D=(X=1,Y=1,Z=1))),RotationOrderForFilter=XZY)
    Child=(Type=Bone,Name="deform_skirt_m_r")
    Filter=(bX=True,bY=True,bZ=True)
    Parents=((Item=(Type=Bone,Name="calf_r"),Weight=1))
    UpAxis=(X=0,Y=-1,Z=0)
    Weight=1
    WorldUp=(Target=(X=0,Y=0,Z=1),Kind=Location,Space=(Type=Bone,Name="pelvis"))
    bMaintainOffset=true
   )
   > &RigUnit_ModifyTransforms(
    ItemToModify="((Item=(Type=Bone,Name=\"deform_skirt_m_l\"),Transform=(Rotation=(X=0,Y=-0.5,Z=0,W=0.866),Translation=(X=0,Y=5,Z=0),Scale3D=(X=1,Y=1,Z=1))),(Item=(Type=Bone,Name=\"deform_skirt_bk_l\"),Transform=(Rotation=(X=0,Y=0,Z=0.4226,W=0.9063),Translation=(X=0,Y=0,Z=0),S ...(truncated)"
    WeightMinimum=0
    Weight=@n11.OutputParam
   )
   > &RigUnit_ModifyTransforms(
    ItemToModify="((Item=(Type=Bone,Name=\"deform_skirt_m_r\"),Transform=(Rotation=(X=-0.0498,Y=-0.2666,Z=0.1032,W=0.957),Translation=(X=-0.6853,Y=-5,Z=1.8344),Scale3D=(X=1,Y=1,Z=1))),(Item=(Type=None,Name=\"deform_skirt_bk_r\"),Transform=(Rotation=(X=0,Y=0,Z=-0.4226,W=0.9063), ...(truncated)"
    WeightMinimum=0
    Weight=@n12.OutputParam
   )
   > &RigUnit_RotationConstraintLocalSpaceOffset(
    AdvancedSettings=(InterpolationType=Average,RotationOrderForFilter=XZY)
    Child=(Type=Bone,Name="deform_skirt_b_l")
    Filter=(bX=True,bY=True,bZ=True)
    Parents=((Item=(Type=Bone,Name="deform_skirt_m_l"),Weight=1))
    Weight=0.2
   )
   > &RigUnit_RotationConstraintLocalSpaceOffset(
    AdvancedSettings=(InterpolationType=Average,RotationOrderForFilter=XZY)
    Child=(Type=Bone,Name="deform_skirt_b_r")
    Filter=(bX=True,bY=True,bZ=True)
    Parents=((Item=(Type=Bone,Name="deform_skirt_m_r"),Weight=1))
    Weight=0.2
   )
   > n13 = &RigUnit_SphericalPoseReader(
    ActiveRegionScaleFactors=(PositiveWidth=3.5,NegativeWidth=2,PositiveHeight=3,NegativeHeight=3)
    ActiveRegionSize=0.1
    Debug=(bDrawDebug=false,DebugScale=25,DebugSegments=20,DebugThickness=0.25)
    DriverAxis=(X=-1,Y=0,Z=0)
    DriverItem=(Type=Bone,Name="thigh_l")
    FalloffRegionScaleFactors=(PositiveWidth=1,NegativeWidth=1,PositiveHeight=1,NegativeHeight=1)
    FalloffSize=0.45
    OptionalParentItem=(Type=Bone)
    RotationOffset=(X=90,Y=0,Z=-30)
   )
   n14 = &RigVMFunction_MathVectorDistance(
    B=&RigUnit_GetTransform(Item=(Type=Bone,Name="root"), bInitial=False).Transform.Translation
    A=&RigUnit_GetTransform(Item=(Type=Bone,Name="pelvis"), bInitial=False).Transform.Translation
   )
   > &RigUnit_ModifyTransforms(
    ItemToModify="((Item=(Type=Bone,Name=\"deform_skirt_bk_r\"),Transform=(Rotation=(X=0,Y=0,Z=-0.1736,W=0.9848),Translation=(X=0,Y=0,Z=0),Scale3D=(X=1,Y=1,Z=1))),(Item=(Type=Bone,Name=\"deform_skirt_b_r\"),Transform=(Rotation=(X=0.2922,Y=-0.1935,Z=-0.2922,W=0.8899),Translation ...(truncated)"
    WeightMinimum=0
    Weight=&RigVMFunction_MathDoubleMul(
     A=&RigVMFunction_MathDoubleRemap(SourceMaximum=45, SourceMinimum=80, TargetMaximum=1, TargetMinimum=0, Value=@n14.Result).Result
     B=@n13.OutputParam
    ).Result
   )
   &RigUnit_ParentConstraint(
    AdvancedSettings=(RotationOrderForFilter=XZY)
    Child=(Type=Bone,Name="deform_phone_l")
    Parents=((Item=(Type=Bone,Name="neck_02"),Weight=0.0241),(Item=(Type=Bone,Name="neck_01"),Weight=0.0649),(Item=(Type=Bone,Name="clavicle_l"),Weight=0.1712),(Item=(Type=Bone,Name="spine_05"),Weight=0.7399))
    Weight=1
    ExecutePin=@RigVMFunction_Sequence.D
   )
   > &RigUnit_ParentConstraint(
    AdvancedSettings=(RotationOrderForFilter=XZY)
    Child=(Type=Bone,Name="deform_phone_r")
    Parents="((Item=(Type=Bone,Name=\"deform_upperarm_fr_r\"),Weight=0.0013),(Item=(Type=Bone,Name=\"upperarm_twist_01_r\"),Weight=0.0014),(Item=(Type=Bone,Name=\"deform_pec_r\"),Weight=0.0015),(Item=(Type=Bone,Name=\"neck_02\"),Weight=0.0034),(Item=(Type=Bone,Name=\"spine_04\") ...(truncated)"
    Weight=1
   )
   > &RigUnit_ParentConstraint(
    AdvancedSettings=(RotationOrderForFilter=XZY)
    Child=(Type=Bone,Name="deform_tie")
    Parents=((Item=(Type=Bone,Name="neck_02"),Weight=0.0033),(Item=(Type=Bone,Name="neck_01"),Weight=0.0331),(Item=(Type=Bone,Name="clavicle_r"),Weight=0.0389),(Item=(Type=Bone,Name="spine_05"),Weight=0.9247))
    Weight=1
   )
   > &RigUnit_AimConstraintLocalSpaceOffset(
    AdvancedSettings=(DebugSettings=(bEnabled=true,Scale=10,WorldOffset=(Rotation=(X=0,Y=0,Z=0,W=1),Translation=(X=0,Y=0,Z=0),Scale3D=(X=1,Y=1,Z=1))),RotationOrderForFilter=XZY)
    AimAxis=(X=-1,Y=0,Z=0)
    Child=(Type=Bone,Name="dyn_phone_1_l")
    Filter=(bX=True,bY=True,bZ=True)
    Parents=((Item=(Type=Bone,Name="dyn_phone_2_r"),Weight=1))
    UpAxis=(X=0,Y=1,Z=0)
    Weight=1
    WorldUp=(Target=(X=0,Y=0,Z=1),Kind=Location,Space=(Type=Bone,Name="dyn_phone_2_r"))
   )
   &RigUnit_ParentConstraint(
    AdvancedSettings=(RotationOrderForFilter=XZY)
    Child=(Type=Bone,Name="deform_collar_r")
    Parents=((Item=(Type=Bone,Name="clavicle_r"),Weight=0.1095),(Item=(Type=Bone,Name="neck_02"),Weight=0.1804),(Item=(Type=Bone,Name="neck_01"),Weight=0.2441),(Item=(Type=Bone,Name="spine_05"),Weight=0.4661))
    Weight=1
    ExecutePin=@RigVMFunction_Sequence.E
   )
   > &RigUnit_ParentConstraint(
    AdvancedSettings=(RotationOrderForFilter=XZY)
    Child=(Type=Bone,Name="deform_collar_l")
    Parents=((Item=(Type=Bone,Name="clavicle_l"),Weight=0.1333),(Item=(Type=Bone,Name="neck_02"),Weight=0.1512),(Item=(Type=Bone,Name="neck_01"),Weight=0.2182),(Item=(Type=Bone,Name="spine_05"),Weight=0.4973))
    Weight=1
   )
   n15 = &RigUnit_GetTransform(Item=(Type=Bone,Name="head"), bInitial=False)
   n16 = &RigVMFunction_MathVectorDistance(
    A=&RigUnit_GetTransform(Item=(Type=Bone,Name="upperarm_twist_01_r"), bInitial=false).Transform.Translation
    B=@n15.Transform.Translation
   )
   n17 = &RigVMFunction_MathQuaternionToEuler(
    RotationOrder=ZYX
    Value=&RigUnit_GetRelativeTransformForItem(Child=(Type=Bone,Name="clavicle_r"), Parent=(Type=Bone,Name="clavicle_r"), bChildInitial=False, bParentInitial=true).RelativeTransform.Rotation
   )
   > &RigUnit_ModifyTransforms(
    ItemToModify=((Item=(Type=Bone,Name="deform_collar_r"),Transform=(Rotation=(X=0.0403,Y=-0.0714,Z=0.4901,W=0.8678),Translation=(X=6.0812,Y=2.6079,Z=0),Scale3D=(X=1.5,Y=1,Z=1))))
    WeightMinimum=0
    Weight=&RigVMFunction_MathDoubleAdd(
     A=&RigVMFunction_MathDoubleRemap(SourceMaximum=11, SourceMinimum=24, TargetMaximum=1, TargetMinimum=0, Value=@n16.Result).Result
     B=&RigVMFunction_MathDoubleRemap(SourceMaximum=-50, SourceMinimum=0, TargetMaximum=1, TargetMinimum=0, Value=@n17.Result.Y).Result
    ).Result
   )
   n18 = &RigUnit_GetTransform(Item=(Type=Bone,Name="upperarm_twist_01_l"), bInitial=false)
   RigUnit_GetRelativeTransformForItem = &RigUnit_GetRelativeTransformForItem(Child=(Type=Bone,Name="clavicle_l"), Parent=(Type=Bone,Name="clavicle_l"), bChildInitial=False, bParentInitial=true)
   RigVMFunction_MathQuaternionToEuler = &RigVMFunction_MathQuaternionToEuler(RotationOrder=ZYX, Value=@RigUnit_GetRelativeTransformForItem.RelativeTransform.Rotation)
   Add = &RigVMFunction_MathDoubleAdd(
    A=&RigVMFunction_MathDoubleRemap(
     SourceMaximum=11
     SourceMinimum=24
     TargetMaximum=1
     TargetMinimum=0
     Value=&RigVMFunction_MathVectorDistance(A=@n18.Transform.Translation, B=@n15.Transform.Translation).Result
    ).Result
    B=&RigVMFunction_MathDoubleRemap(SourceMaximum=-50, SourceMinimum=0, TargetMaximum=1, TargetMinimum=0, Value=@RigVMFunction_MathQuaternionToEuler.Result.Y).Result
   )
   > &RigUnit_ModifyTransforms(
    ItemToModify=((Item=(Type=Bone,Name="deform_collar_l"),Transform=(Rotation=(X=-0.0399,Y=-0.1546,Z=0.3171,W=0.9349),Translation=(X=-5.3384,Y=-3.0624,Z=0),Scale3D=(X=1.862,Y=1,Z=1))))
    WeightMinimum=0
    Weight=@Add.Result
   )
   Remap = &RigVMFunction_MathDoubleRemap(
    SourceMaximum=20
    SourceMinimum=0
    TargetMaximum=1
    TargetMinimum=0
    bClamp=False
    Value=&RigUnit_GetRelativeTransformForItem(Child=(Type=Bone,Name="deform_pec_r"), Parent=(Type=Bone,Name="deform_pec_r"), bChildInitial=false, bParentInitial=true).RelativeTransform.Translation.X
   )
   > &RigUnit_ModifyTransforms(
    ItemToModify=((Item=(Type=Bone,Name="deform_phone_r"),Transform=(Rotation=(X=0,Y=0,Z=0,W=1),Translation=(X=0,Y=-1.5,Z=0),Scale3D=(X=1,Y=1,Z=1))))
    WeightMinimum=0
    Weight=@Remap.Result
   )
   &RigUnit_ParentConstraint(
    AdvancedSettings=(RotationOrderForFilter=XZY)
    Child=(Type=Bone,Name="deform_laces_r")
    Parents=((Item=(Type=Bone,Name="thigh_r"),Weight=0.0036),(Item=(Type=Bone,Name="deform_knee_out_r"),Weight=0.0074),(Item=(Type=Bone,Name="calf_twist_02_r"),Weight=0.4519),(Item=(Type=Bone,Name="calf_r"),Weight=0.5371))
    Weight=1
    ExecutePin=@RigVMFunction_Sequence.F
   )
   > &RigUnit_ParentConstraint(
    AdvancedSettings=(RotationOrderForFilter=XZY)
    Child=(Type=Bone,Name="deform_laces_l")
    Parents=((Item=(Type=Bone,Name="thigh_l"),Weight=0.0017),(Item=(Type=Bone,Name="deform_knee_out_l"),Weight=0.0054),(Item=(Type=Bone,Name="calf_l"),Weight=0.4597),(Item=(Type=Bone,Name="calf_twist_02_l"),Weight=0.5333))
    Weight=1
   )
   &RigUnit_AimConstraintLocalSpaceOffset(
    AdvancedSettings=(DebugSettings=(bEnabled=true,Scale=10,WorldOffset=(Rotation=(X=0,Y=0,Z=0,W=1),Translation=(X=0,Y=0,Z=0),Scale3D=(X=1,Y=1,Z=1))),RotationOrderForFilter=XZY)
    AimAxis=(X=-1,Y=0,Z=0)
    Child=(Type=Bone,Name="deform_skirt_bk_l")
    Filter=(bX=True,bY=True,bZ=True)
    Parents=((Item=(Type=Bone,Name="foot_l"),Weight=1))
    UpAxis=(X=0,Y=1,Z=0)
    Weight=1
    WorldUp=(Target=(X=0,Y=0,Z=1),Kind=Location,Space=(Type=Bone,Name="deform_skirt_bk_l"))
    bMaintainOffset=true
    ExecutePin=@RigVMFunction_Sequence.G
   )
   > &RigUnit_AimConstraintLocalSpaceOffset(
    AdvancedSettings=(DebugSettings=(bEnabled=true,Scale=10,WorldOffset=(Rotation=(X=0,Y=0,Z=0,W=1),Translation=(X=0,Y=0,Z=0),Scale3D=(X=1,Y=1,Z=1))),RotationOrderForFilter=XZY)
    Child=(Type=Bone,Name="deform_skirt_bk_r")
    Filter=(bX=True,bY=True,bZ=True)
    Parents=((Item=(Type=Bone,Name="foot_r"),Weight=1))
    UpAxis=(X=0,Y=-1,Z=0)
    Weight=1
    WorldUp=(Target=(X=0,Y=0,Z=1),Kind=Location,Space=(Type=Bone,Name="deform_skirt_bk_r"))
    bMaintainOffset=true
   )
   AlphaInterp = &RigVMFunction_AlphaInterp(InterpSpeedDecreasing=10, InterpSpeedIncreasing=5, Scale=1, bClampResult=False, Value=RigVMVariableNode() {variable=IsKneeSliding}.Value)
   > &RigUnit_ModifyTransforms(
    ItemToModify="((Item=(Type=Bone,Name=\"deform_skirt_b_l\"),Transform=(Rotation=(X=0.3891,Y=-0.4015,Z=-0.2992,W=0.7732),Translation=(X=-0.0163,Y=2.4901,Z=-2.2171),Scale3D=(X=1,Y=1,Z=1))),(Item=(Type=Bone,Name=\"deform_skirt_b_r\"),Transform=(Rotation=(X=0.1033,Y=-0.4789,Z=-0 ...(truncated)"
    WeightMinimum=0
    Weight=@AlphaInterp.Result
   )
   &RigUnit_ParentConstraint(
    AdvancedSettings=(RotationOrderForFilter=XZY)
    Child=(Type=Bone,Name="deform_skirt_m_l")
    ExecutePin=()
    Parents="((Item=(Type=Bone,Name=\"thigh_l\"),Weight=0.0032),(Item=(Type=Bone,Name=\"spine_03\"),Weight=0.0042),(Item=(Type=Bone,Name=\"spine_01\"),Weight=0.008),(Item=(Type=Bone,Name=\"deform_glute_l\"),Weight=0.0093),(Item=(Type=Bone,Name=\"spine_02\"),Weight=0.032),(Item=( ...(truncated)"
    Weight=1
   )
   RigVMFunction_MathVectorFromDouble = /Script/RigVM.RigVMFunction_MathVectorFromDouble(Value=@RigVMFunction_MathQuaternionToEuler.Result.Y) {is=struct RigVMUnitNode, template="Cast::Execute(in Value,out Result)"}
  ]}
]
hierarchy [
 root {loc=<0 0 0>, rot=<0 0 0 1>}
 pelvis {parent=root, loc=<0 -0.5301 98.0571>, rot=<0.0127 0.707 -0.0127 -0.707>}
 spine_01 {parent=pelvis, loc=<6.1208 -0.9328 0>, rot=<0 0 0.034 -0.9994>}
 spine_02 {parent=spine_01, loc=<6.5997 -0.2095 0>, rot=<0 0 0.0238 -0.9997>}
 spine_03 {parent=spine_02, loc=<6.4914 -0.1237 0>, rot=<0 0 0.0149 -0.9999>}
 spine_04 {parent=spine_03, loc=<6.9598 0.0794 0>, rot=<0 0 -0.0132 -0.9999>}
 spine_05 {parent=spine_04, loc=<6.9987 0 0>, rot=<0 0 -0.0015 -1>}
 attach_cape {parent=spine_05, loc=<9.3216 -8.4501 0>, rot=<0.0616 -0.7044 -0.0616 -0.7044>}
 neck_01 {parent=spine_05, loc=<11.3138 0.1529 0>, rot=<0 0 -0.2003 -0.9797>}
 neck_02 {parent=neck_01, loc=<4.5627 0 0>, rot=<0 0 0.0504 -0.9987>}
 head {parent=neck_02, loc=<5.3914 0 0>, rot=<0 0 0.115 -0.9934>}
 attach_backpack {parent=spine_05, loc=<-0.2609 -18.6936 0.009>, rot=<-0.0087 -0.7069 0.0092 -0.7072>}
 clavicle_l {parent=spine_05, loc=<5.4831 -0.1053 -3.2946>, rot=<0.0465 -0.7289 -0.009 -0.6829>}
 upperarm_l {parent=clavicle_l, loc=<9.8658 -1.2136 -0.4105>, rot=<0.0164 -0.3245 -0.0269 -0.9454>}
 lowerarm_l {parent=upperarm_l, loc=<26.465 -0.8345 0.259>, rot=<0.0036 0.0161 -0.2298 -0.9731>}
 hand_l {parent=lowerarm_l, loc=<22.7355 -0.3203 -0.6105>, rot=<-0.6784 -0.0223 0.0735 -0.7307>}
 hand_attach_l {parent=hand_l, loc=<8.5 -3 0>, rot=<0.7071 0.7071 0 0>}
 weapon_l {parent=hand_l, loc=<-0.1974 -3.4111 0.3811>, rot=<0 0 0 1>}
 thumb_01_l {parent=hand_l, loc=<1.3118 -0.3191 -2.459>, rot=<0.6546 -0.4042 0.0664 -0.6354>}
 thumb_02_l {parent=thumb_01_l, loc=<5.9336 -0.0448 0.0401>, rot=<-0.0037 0.0213 0.298 -0.9543>}
 thumb_03_l {parent=thumb_02_l, loc=<2.2223 0.3682 -0.0907>, rot=<0.0118 -0.013 0.0066 -0.9998>}
 index_metacarpal_l {parent=hand_l, loc=<1.6535 0.0183 -2.96>, rot=<0.0158 -0.0124 -0.0173 -0.9996>}
 index_01_l {parent=index_metacarpal_l, loc=<8.3008 0.0357 -0.0524>, rot=<-0.0044 0.0485 0.0924 -0.9945>}
 index_02_l {parent=index_01_l, loc=<3.3587 -0.3004 0.0452>, rot=<-0.002 0.0027 0.3225 -0.9466>}
 index_03_l {parent=index_02_l, loc=<2.5578 -0.123 -0.0252>, rot=<-0.002 0.0067 0.2558 -0.9667>}
 middle_metacarpal_l {parent=hand_l, loc=<2.1856 -0.1725 -0.5984>, rot=<0.006 0.0168 -0.0329 -0.9993>}
 middle_01_l {parent=middle_metacarpal_l, loc=<7.3275 -0.2447 0.0129>, rot=<-0.0208 0.0042 0.0858 -0.9961>}
 middle_02_l {parent=middle_01_l, loc=<4.0123 0.1401 0.1175>, rot=<-0.0218 0.0258 0.3512 -0.9357>}
 middle_03_l {parent=middle_02_l, loc=<2.6077 0.0534 0.1373>, rot=<0.0082 0.0188 0.1708 -0.9851>}
 ring_metacarpal_l {parent=hand_l, loc=<2.2074 -0.3625 1.3638>, rot=<-0.0284 0.0722 -0.0458 -0.9959>}
 ring_01_l {parent=ring_metacarpal_l, loc=<7.0295 -0.6197 -0.1446>, rot=<-0.0489 -0.0252 0.1892 -0.9804>}
 ring_02_l {parent=ring_01_l, loc=<3.6174 0.4647 -0.1027>, rot=<0 0 0.2806 -0.9598>}
 ring_03_l {parent=ring_02_l, loc=<2.4152 0.114 0.0199>, rot=<0 0 0.1189 -0.9929>}
 pinky_metacarpal_l {parent=hand_l, loc=<1.7616 -0.9753 3.3814>, rot=<-0.0397 0.0814 -0.069 -0.9935>}
 pinky_01_l {parent=pinky_metacarpal_l, loc=<7.1709 -0.4831 -0.1094>, rot=<-0.0654 -0.0035 0.2242 -0.9723>}
 pinky_02_l {parent=pinky_01_l, loc=<2.8274 0.0893 -0.1043>, rot=<0.0037 0.0011 0.2854 -0.9584>}
 pinky_03_l {parent=pinky_02_l, loc=<2.2046 0.0307 -0.0078>, rot=<-0.0002 0.0098 0.1445 -0.9895>}
 deform_wrist_dn_base_l {parent=hand_l, loc=<0 -1.5326 0>, rot=<0 0 0 -1>}
 deform_wrist_dn_l {parent=deform_wrist_dn_base_l, loc=<0 0 0>, rot=<0 0 0 1>}
 deform_wrist_up_base_l {parent=hand_l, loc=<0 2.3009 0>, rot=<0 0 0 -1>}
 deform_wrist_up_l {parent=deform_wrist_up_base_l, loc=<0 0 0>, rot=<0 0 0 1>}
 lowerarm_twist_02_l {parent=lowerarm_l, loc=<10.6707 0 0>, rot=<0 0 0 1>}
 lowerarm_twist_01_l {parent=lowerarm_l, loc=<18.316 0 0>, rot=<0 0 0 1>}
 deform_elbow_in_base_l {parent=lowerarm_l, loc=<0.1834 3.6062 -0.2572>, rot=<0 0 0 -1>}
 deform_elbow_in_l {parent=deform_elbow_in_base_l, loc=<0 0 0>, rot=<0 0 0 1>}
 deform_elbow_out_base_l {parent=lowerarm_l, loc=<0 -4 0>, rot=<0 0 0 1>}
 deform_elbow_out_l {parent=deform_elbow_out_base_l, loc=<0 0 0>, rot=<0 0 0 1>}
 upperarm_twist_01_l {parent=upperarm_l, loc=<7.2135 -0.1134 0>, rot=<0 0 0 1>}
 deform_upperarm_bk_base_l {parent=upperarm_twist_01_l, loc=<-6.4053 -4.3746 0.314>, rot=<0 0 0 -1>}
 deform_upperarm_bk_l {parent=deform_upperarm_bk_base_l, loc=<0 0 0>, rot=<0 0 0 1>}
 deform_upperarm_fr_base_l {parent=upperarm_twist_01_l, loc=<-5.6312 5.092 0.3354>, rot=<0 0 0 -1>}
 deform_upperarm_fr_l {parent=deform_upperarm_fr_base_l, loc=<0 0 0>, rot=<0 0 0 1>}
 deform_upperarm_up_base_l {parent=upperarm_twist_01_l, loc=<-8.891 0.4825 3.9517>, rot=<0 0 0 -1>}
 deform_upperarm_up_l {parent=deform_upperarm_up_base_l, loc=<0 0 0>, rot=<0 0 0 1>}
 upperarm_twist_02_l {parent=upperarm_l, loc=<19.7998 -0.1134 0>, rot=<0 0 0 1>}
 clavicle_r {parent=spine_05, loc=<5.4831 -0.1053 3.2946>, rot=<0.7289 0.0465 -0.6829 0.009>}
 upperarm_r {parent=clavicle_r, loc=<-9.8658 1.2136 0.4105>, rot=<0.0164 -0.3245 -0.0269 -0.9454>}
 lowerarm_r {parent=upperarm_r, loc=<-26.465 0.8345 -0.259>, rot=<0.0036 0.0161 -0.2298 -0.9731>}
 hand_r {parent=lowerarm_r, loc=<-22.7355 0.3203 0.6106>, rot=<-0.6784 -0.0223 0.0735 -0.7307>}
 thumb_01_r {parent=hand_r, loc=<-1.3118 0.3191 2.459>, rot=<0.6546 -0.4042 0.0664 -0.6354>}
 thumb_02_r {parent=thumb_01_r, loc=<-5.9336 0.0448 -0.0401>, rot=<-0.0037 0.0213 0.298 -0.9543>}
 thumb_03_r {parent=thumb_02_r, loc=<-2.2223 -0.3682 0.0907>, rot=<0.0118 -0.013 0.0066 -0.9998>}
 index_metacarpal_r {parent=hand_r, loc=<-1.6535 -0.0183 2.96>, rot=<0.0158 -0.0124 -0.0173 -0.9996>}
 index_01_r {parent=index_metacarpal_r, loc=<-8.3008 -0.0357 0.0524>, rot=<-0.0044 0.0485 0.0924 -0.9945>}
 index_02_r {parent=index_01_r, loc=<-3.3587 0.3004 -0.0452>, rot=<-0.002 0.0027 0.3225 -0.9466>}
 index_03_r {parent=index_02_r, loc=<-2.5578 0.123 0.0252>, rot=<-0.002 0.0067 0.2558 -0.9667>}
 middle_metacarpal_r {parent=hand_r, loc=<-2.1856 0.1725 0.5984>, rot=<0.006 0.0168 -0.0329 -0.9993>}
 middle_01_r {parent=middle_metacarpal_r, loc=<-7.3275 0.2447 -0.0129>, rot=<-0.0208 0.0042 0.0858 -0.9961>}
 middle_02_r {parent=middle_01_r, loc=<-4.0123 -0.1401 -0.1175>, rot=<-0.0218 0.0258 0.3512 -0.9357>}
 middle_03_r {parent=middle_02_r, loc=<-2.6077 -0.0534 -0.1373>, rot=<0.0082 0.0188 0.1708 -0.9851>}
 ring_metacarpal_r {parent=hand_r, loc=<-2.2074 0.3625 -1.3638>, rot=<-0.0284 0.0722 -0.0458 -0.9959>}
 ring_01_r {parent=ring_metacarpal_r, loc=<-7.0295 0.6197 0.1446>, rot=<-0.0489 -0.0252 0.1892 -0.9804>}
 ring_02_r {parent=ring_01_r, loc=<-3.6174 -0.4647 0.1027>, rot=<0 0 0.2806 -0.9598>}
 ring_03_r {parent=ring_02_r, loc=<-2.4152 -0.114 -0.0199>, rot=<0 0 0.1189 -0.9929>}
 pinky_metacarpal_r {parent=hand_r, loc=<-1.7616 0.9753 -3.3814>, rot=<-0.0397 0.0814 -0.069 -0.9935>}
 pinky_01_r {parent=pinky_metacarpal_r, loc=<-7.1709 0.4831 0.1094>, rot=<-0.0654 -0.0035 0.2242 -0.9723>}
 pinky_02_r {parent=pinky_01_r, loc=<-2.8274 -0.0893 0.1043>, rot=<0.0037 0.0011 0.2854 -0.9584>}
 pinky_03_r {parent=pinky_02_r, loc=<-2.2046 -0.0307 0.0078>, rot=<-0.0002 0.0098 0.1445 -0.9895>}
 weapon_r {parent=hand_r, loc=<0.1974 3.4111 -0.3811>, rot=<0 0 0 1>}
 hand_attach_r {parent=hand_r, loc=<-8.5001 2.9999 0>, rot=<0 0 -0.7071 -0.7071>}
 deform_wrist_dn_base_r {parent=hand_r, loc=<-0.0001 1.5328 0>, rot=<0 0 0 -1>}
 deform_wrist_dn_r {parent=deform_wrist_dn_base_r, loc=<0 0 0>, rot=<0 0 0 1>}
 deform_wrist_up_base_r {parent=hand_r, loc=<-0.0002 -2.3007 -0.0001>, rot=<0 0 0 -1>}
 deform_wrist_up_r {parent=deform_wrist_up_base_r, loc=<0 0 0>, rot=<0 0 0 1>}
 lowerarm_twist_02_r {parent=lowerarm_r, loc=<-10.6707 0 0>, rot=<0 0 0 -1>}
 lowerarm_twist_01_r {parent=lowerarm_r, loc=<-18.316 0 0>, rot=<0 0 0 -1>}
 deform_elbow_in_base_r {parent=lowerarm_r, loc=<-0.1834 -3.6062 0.2572>, rot=<0 0 0 -1>}
 deform_elbow_in_r {parent=deform_elbow_in_base_r, loc=<0 0 0>, rot=<0 0 0 1>}
 deform_elbow_out_base_r {parent=lowerarm_r, loc=<0 4 -0.0001>, rot=<0 0 0 -1>}
 deform_elbow_out_r {parent=deform_elbow_out_base_r, loc=<0 0 0>, rot=<0 0 0 1>}
 upperarm_twist_01_r {parent=upperarm_r, loc=<-7.2135 0.1134 0>, rot=<0 0 0 -1>}
 deform_upperarm_bk_base_r {parent=upperarm_twist_01_r, loc=<6.4053 4.3746 -0.314>, rot=<0 0 0 -1>}
 deform_upperarm_bk_r {parent=deform_upperarm_bk_base_r, loc=<0 0 0>, rot=<0 0 0 1>}
 deform_upperarm_fr_base_r {parent=upperarm_twist_01_r, loc=<5.6312 -5.092 -0.3354>, rot=<0 0 0 -1>}
 deform_upperarm_fr_r {parent=deform_upperarm_fr_base_r, loc=<0 0 0>, rot=<0 0 0 1>}
 deform_upperarm_up_base_r {parent=upperarm_twist_01_r, loc=<8.891 -0.4825 -3.9517>, rot=<0 0 0 -1>}
 deform_upperarm_up_r {parent=deform_upperarm_up_base_r, loc=<0 0 0>, rot=<0 0 0 1>}
 upperarm_twist_02_r {parent=upperarm_r, loc=<-19.7998 0.1134 0>, rot=<0 0 0 -1>}
 deform_collar_l {parent=spine_05, loc=<11.7856 0.3673 -4.6212>, rot=<-0.5138 0.4685 0.6755 0.2453>}
 dyn_collar_l {parent=deform_collar_l, loc=<0 0 0>, rot=<0 0 0 1>}
 deform_collar_r {parent=spine_05, loc=<11.7855 0.3673 4.6212>, rot=<-0.4685 -0.5138 0.2453 -0.6755>}
 dyn_collar_r {parent=deform_collar_r, loc=<0 0 0>, rot=<0 0 0 1>}
 deform_tie {parent=spine_05, loc=<6.8704 5.5647 0>, rot=<0 0 0.9346 0.3558>}
 dyn_tie {parent=deform_tie, loc=<0 0 0>, rot=<0 -0.0005 -0.0306 -0.9995>}
 dyn_tie_1 {parent=dyn_tie, loc=<6.3906 0 0>, rot=<-0.015 0.0014 -0.0956 -0.9953>}
 dyn_tie_2 {parent=dyn_tie_1, loc=<10.5479 0 0>, rot=<0 0 -0.1094 -0.994>}
 deform_phone_r {parent=spine_05, loc=<7.3706 4.2822 5.1474>, rot=<-0.0685 -0.0349 0.9269 0.3673>}
 dyn_phone_r {parent=deform_phone_r, loc=<0 0 0>, rot=<-0.0569 -0.0114 -0.0095 -0.9983>}
 dyn_phone_1_r {parent=dyn_phone_r, loc=<6.4359 0 0>, rot=<0.04 -0.0565 -0.0546 -0.9961>}
 dyn_phone_2_r {parent=dyn_phone_1_r, loc=<9.2301 0 0>, rot=<0.0386 0.09 -0.1491 -0.984>}
 deform_phone_l {parent=spine_05, loc=<7.7139 4.4821 -4.2085>, rot=<0.1209 -0.0372 0.9139 0.3858>}
 dyn_phone_l {parent=deform_phone_l, loc=<0 0 0>, rot=<0.0082 0.0098 -0.0184 -0.9997>}
 dyn_phone_1_l {parent=dyn_phone_l, loc=<7.2616 0 0>, rot=<-0.0719 0.0967 -0.0951 -0.9881>}
 deform_pec_aim_base_l {parent=spine_04, loc=<3.2456 9.3309 -1.2237>, rot=<0.0603 -0.4033 0.3318 -0.8507>}
 deform_pec_aim_l {parent=deform_pec_aim_base_l, loc=<0 0 0>, rot=<0 0 0 1>}
 deform_pec_base_l {parent=deform_pec_aim_l, loc=<9.3665 0 0>, rot=<0 0 0 1>}
 deform_pec_l {parent=deform_pec_base_l, loc=<0 0 0>, rot=<0 0 0 1>}
 deform_pec_aim_base_r {parent=spine_04, loc=<3.2452 9.3309 1.2237>, rot=<-0.4033 -0.0603 0.8507 0.3318>}
 deform_pec_aim_r {parent=deform_pec_aim_base_r, loc=<0 0 0>, rot=<0 0 0 1>}
 deform_pec_base_r {parent=deform_pec_aim_r, loc=<-9.3667 -0.0003 -0.0004>, rot=<0 0 0 1>}
 deform_pec_r {parent=deform_pec_base_r, loc=<0 0 0>, rot=<0 0 0 1>}
 deform_hoodie_l {parent=spine_01, loc=<2.8558 11.1568 -0.2296>, rot=<-0.054 -0.0058 0.9972 0.0509>}
 dyn_hoodie_l {parent=deform_hoodie_l, loc=<0 0 0>, rot=<0 0 0 1>}
 dyn_hoodie_1_l {parent=dyn_hoodie_l, loc=<12.0052 0 0>, rot=<0.002 0.0334 -0.0601 -0.9976>}
 deform_skirt_b_r {parent=spine_01, loc=<-4.3317 -5.0728 13.1331>, rot=<0.173 -0.8157 0.5181 0.1905>}
 dyn_skirt_b_r {parent=deform_skirt_b_r, loc=<0 0 0>, rot=<0 0 0 1>}
 dyn_hoodie_b_r {parent=dyn_skirt_b_r, loc=<0 -1.7692 0>, rot=<0.1894 0.136 -0.0093 -0.9724>}
 dyn_hoodie_b_1_r {parent=dyn_hoodie_b_r, loc=<17.4753 0 0>, rot=<0.0982 0.0187 -0.1861 -0.9774>}
 deform_skirt_f_r {parent=spine_01, loc=<-3.0596 7.9597 7.1715>, rot=<0.0593 -0.1412 0.9851 0.0778>}
 dyn_skirt_f_r {parent=deform_skirt_f_r, loc=<0 0 0>, rot=<0 0 0 1>}
 deform_hoodie_r {parent=dyn_skirt_f_r, loc=<-0.3962 -1.2384 -1.0365>, rot=<-0.0067 -0.0173 0.0279 -0.9994>}
 dyn_hoodie_r {parent=deform_hoodie_r, loc=<0 0 0>, rot=<0 0 0 1>}
 dyn_hoodie_1_r {parent=dyn_hoodie_r, loc=<11.7707 0 0>, rot=<-0.0005 -0.0391 -0.0131 -0.9991>}
 dyn_skirt_f_r1 {parent=dyn_skirt_f_r, loc=<12.9417 0 0>, rot=<0 0 0 1>}
 deform_skirt_f_r_aim {parent=dyn_skirt_f_r1, loc=<12.6913 0 0>, rot=<0 0 0 1>}
 deform_skirt_bk_r {parent=spine_01, loc=<-6.6564 -10.3623 6.4919>, rot=<0.0794 -0.9928 0.0713 0.0548>}
 dyn_skirt_bk_r {parent=deform_skirt_bk_r, loc=<0 0 0>, rot=<0 0 0 1>}
 dyn_hoodie_bk_r {parent=dyn_skirt_bk_r, loc=<0 -1.7692 0>, rot=<0 0 0.0565 -0.9984>}
 dyn_hoodie_bk_1_r {parent=dyn_hoodie_bk_r, loc=<16.4545 0.5174 0>, rot=<0 0 -0.1034 -0.9946>}
 deform_skirt_m_r {parent=spine_01, loc=<-2.8945 3.4837 13.0344>, rot=<0.139 -0.4984 0.8388 0.1691>}
 dyn_skirt_m_r {parent=deform_skirt_m_r, loc=<0 0 0>, rot=<0 0 0 1>}
 dyn_skirt_m_r1 {parent=dyn_skirt_m_r, loc=<12.9417 0 0>, rot=<0 0 0 1>}
 deform_skirt_m_r_aim {parent=dyn_skirt_m_r1, loc=<11.6975 0.3696 0.156>, rot=<-0.253 0 0 -0.9675>}
 deform_skirt_m_l {parent=spine_01, loc=<-2.8942 3.4837 -13.0344>, rot=<0.4984 0.139 0.1691 -0.8388>}
 dyn_skirt_m_l {parent=deform_skirt_m_l, loc=<0 0 0>, rot=<0 0 0 1>}
 dyn_skirt_m_1_l {parent=dyn_skirt_m_l, loc=<-12.4533 0 0>, rot=<0 0 0 1>}
 deform_skirt_chain_f_l {parent=deform_skirt_m_l, loc=<-2.0591 0.2329 -0.101>, rot=<0.0286 -0.0458 0.8724 0.4858>}
 dyn_skirt_keychain_f_l {parent=deform_skirt_chain_f_l, loc=<0 0 0>, rot=<-0.0051 -0.0057 0.0025 -1>}
 deform_skirt_f_l {parent=spine_01, loc=<-3.0597 7.9597 -7.1715>, rot=<0.1412 0.0593 0.0778 -0.9851>}
 dyn_skirt_f_l {parent=deform_skirt_f_l, loc=<0 0 0>, rot=<0 0 0 1>}
 dyn_skirt_f_1_l {parent=dyn_skirt_f_l, loc=<-12.4533 0 0>, rot=<0 0 0 1>}
 deform_skirt_f_l_aim {parent=dyn_skirt_f_1_l, loc=<-13.1801 0 0>, rot=<0 0 0 1>}
 deform_skirt_chain_f_r {parent=dyn_skirt_f_l, loc=<-6.4611 0.5281 0>, rot=<0 0 0 1>}
 dyn_skirt_keychain_f_r {parent=dyn_skirt_f_l, loc=<-1.9148 1.2435 0.4436>, rot=<0.0929 0.0514 0.0534 -0.9929>}
 deform_skirt_bk_l {parent=spine_01, loc=<-6.6564 -10.3623 -6.4919>, rot=<0.9928 0.0794 0.0548 -0.0713>}
 dyn_skirt_bk_l {parent=deform_skirt_bk_l, loc=<0 0 0>, rot=<0 0 0 -1>}
 dyn_hoodie_bk_l {parent=dyn_skirt_bk_l, loc=<0 1.7692 0>, rot=<0 0 0.0565 -0.9984>}
 dyn_hoodie_bk_1_l {parent=dyn_hoodie_bk_l, loc=<-16.4545 -0.5173 0>, rot=<0 0 -0.1034 -0.9946>}
 deform_skirt_b_l {parent=spine_01, loc=<-4.3317 -5.0728 -13.1331>, rot=<0.8157 0.173 0.1905 -0.5181>}
 dyn_skirt_b_l {parent=deform_skirt_b_l, loc=<0 0 0>, rot=<0 0 0 1>}
 dyn_hoodie_b_l {parent=dyn_skirt_b_l, loc=<0.0003 1.7694 0>, rot=<0.1894 0.136 -0.0093 -0.9724>}
 dyn_hoodie_b_1_l {parent=dyn_hoodie_b_l, loc=<-17.4757 -0.0002 -0.0001>, rot=<0.0982 0.0187 -0.1861 -0.9774>}
 deform_skirt_f_aim {parent=spine_01, loc=<-29.0525 11.7735 -0.1181>, rot=<-0.0011 0.0077 -0.0132 -0.9999>}
 thigh_l {parent=pelvis, loc=<-1.0855 -0.2109 -8.7053>, rot=<0.0009 0.0355 0.0155 -0.9993>}
 calf_l {parent=thigh_l, loc=<-42.5284 -0.0019 -0.7297>, rot=<0.0006 -0.0161 -0.0391 -0.9991>}
 foot_l {parent=calf_l, loc=<-46.7401 -2.5014 -0.592>, rot=<0.0427 -0.0238 0.0353 -0.9982>}
 ball_l {parent=foot_l, loc=<-3.5658 10.8066 -0.1544>, rot=<0.0001 -0.0001 -0.7063 -0.7079>}
 calf_twist_02_l {parent=calf_l, loc=<-19.1546 -1.7943 -0.2638>, rot=<0.0237 0.0095 -0.0035 -0.9997>}
 calf_twist_01_l {parent=calf_l, loc=<-35.8483 -1.9197 -0.5794>, rot=<0.0237 0.0095 -0.0035 -0.9997>}
 deform_knee_in_base_l {parent=calf_l, loc=<0.8031 -5.0231 0.2456>, rot=<0 0 0 -1>}
 deform_knee_in_l {parent=deform_knee_in_base_l, loc=<0 0 0>, rot=<0 0 0 1>}
 deform_knee_out_base_l {parent=calf_l, loc=<0 4 0>, rot=<0 0 0 -1>}
 deform_knee_out_l {parent=deform_knee_out_base_l, loc=<0 0 0>, rot=<0 0 0 1>}
 deform_laces_l {parent=calf_l, loc=<-14.6902 -2.8362 -6.0982>, rot=<0.0043 0.0752 -0.0564 -0.9956>}
 dyn_laces_l {parent=deform_laces_l, loc=<0 0 0>, rot=<0 0 0 1>}
 thigh_twist_01_l {parent=thigh_l, loc=<-27.952 0 0>, rot=<0.0013 0 0 -1>}
 deform_groin_base_l {parent=thigh_l, loc=<0.4801 5.6776 -0.2486>, rot=<0 0 0 -1>}
 deform_groin_l {parent=deform_groin_base_l, loc=<0 0 0>, rot=<0 0 0 1>}
 thigh_r {parent=pelvis, loc=<-1.0855 -0.2109 8.7053>, rot=<0.0355 -0.0009 0.9993 0.0155>}
 calf_r {parent=thigh_r, loc=<42.5284 0.0019 0.7297>, rot=<0.0006 -0.0161 -0.0391 -0.9991>}
 foot_r {parent=calf_r, loc=<46.7401 2.5014 0.592>, rot=<0.0427 -0.0238 0.0353 -0.9982>}
 ball_r {parent=foot_r, loc=<3.5658 -10.8066 0.1544>, rot=<0.0001 -0.0001 -0.7063 -0.7079>}
 calf_twist_02_r {parent=calf_r, loc=<19.1546 1.7943 0.2638>, rot=<0.0237 0.0095 -0.0035 -0.9997>}
 calf_twist_01_r {parent=calf_r, loc=<35.8483 1.9197 0.5794>, rot=<0.0237 0.0095 -0.0035 -0.9997>}
 deform_knee_in_base_r {parent=calf_r, loc=<-0.8031 5.0231 -0.2457>, rot=<0 0 0 -1>}
 deform_knee_in_r {parent=deform_knee_in_base_r, loc=<0 0 0>, rot=<0 0 0 1>}
 deform_knee_out_base_r {parent=calf_r, loc=<0 -4 0>, rot=<0 0 0 -1>}
 deform_knee_out_r {parent=deform_knee_out_base_r, loc=<0 0 0>, rot=<0 0 0 1>}
 deform_laces_r {parent=calf_r, loc=<13.3347 2.6834 6.0704>, rot=<0.0043 0.0752 -0.0564 -0.9956>}
 dyn_laces_r {parent=deform_laces_r, loc=<0 0 0>, rot=<0 0 0 -1>}
 thigh_twist_01_r {parent=thigh_r, loc=<27.952 0 0>, rot=<0.0013 0 0 -1>}
 deform_groin_base_r {parent=thigh_r, loc=<-0.4801 -5.6776 0.2486>, rot=<0 0 0 -1>}
 deform_groin_r {parent=deform_groin_base_r, loc=<0 0 0>, rot=<0 0 0 1>}
 deform_glute_base_l {parent=pelvis, loc=<-7.3466 -9.0213 -9.1245>, rot=<0.0172 0.031 -0.4477 -0.8935>}
 deform_glute_l {parent=deform_glute_base_l, loc=<0 0 0>, rot=<0 0 0 1>}
 deform_glute_base_r {parent=pelvis, loc=<-7.3464 -9.0213 9.1245>, rot=<0.031 -0.0172 0.8935 -0.4477>}
 deform_glute_r {parent=deform_glute_base_r, loc=<0 0 0>, rot=<0 0 0 1>}
]
variables [
 IsKneeSliding {type=float, default=0}
 Skydiving_Alpha {type=bool, default=False}
]
