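# Physics asset for the F_MED_PocketScrunchie female character body, setting up secondary motion for its skirt, hoodie and keychains.
# Capsules on pelvis, spine and legs, plus a large root sphere, form the body shapes.
# Garment chains mostly start at free-rotation deform bones on spine_01, then use tight swing/twist limits, though front skirt joints stay free.
# Rear hoodie chains hang from the skirt chains, with free-rotation cross-links joining their tips.
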PhysicsAsset /BRCosmetics/Characters/Player/Female/Medium/Bodies/F_MED_PocketScrunchie/Meshes/F_MED_PocketScrunchie_Physics
bodies {
 calf_l [{shapeType=Capsule, length=43.5914, radius=11.0533, shapeName=calf_l_capsule, center=<-34.1264 -2.5014 0.248>, rotation=<rot 90 0 180>}]
 calf_r [{shapeType=Capsule, length=43.5914, radius=11.0533, shapeName=calf_l_capsule, center=<34.1264 2.5014 -0.248>, rotation=<rot 90 0 0>}]
 deform_hoodie_l [{shapeType=Capsule, length=1, radius=1, shapeName=deform_hoodie_l_capsule}]
 deform_hoodie_r [{shapeType=Capsule, length=1, radius=1, shapeName=deform_hoodie_r_capsule, rotation=<rot 90 0 0>}]
 deform_skirt_b_l [{shapeType=Capsule, length=1, radius=1, shapeName=deform_skirt_b_l_capsule}]
 deform_skirt_b_r [{shapeType=Capsule, length=1, radius=1, shapeName=deform_skirt_b_r_capsule}]
 deform_skirt_bk_l [{shapeType=Capsule, length=1, radius=1, shapeName=deform_skirt_bk_l_capsule}]
 deform_skirt_bk_r [{shapeType=Capsule, length=1, radius=1, shapeName=deform_skirt_bk_r_capsule}]
 deform_skirt_chain_f_l [{shapeType=Capsule, length=8.7426, radius=3.2363, shapeName=deform_skirt_chain_f_l_capsule, center=<-0.4448 4.0945 0>}]
 deform_skirt_chain_f_r [{
   shapeType=Capsule
   length=8.5213
   radius=3.8447
   shapeName=deform_skirt_chain_f_r_capsule
   center=<0 -2.5649 -0.0355>
   rotation=<rot -0.0001 -0.0002 3.0092>
  }]
 deform_skirt_f_aim [{shapeType=Capsule, length=5.0585, radius=10, shapeName=deform_skirt_f_aim_capsule, center=<2.3481 -8.096 0>}]
 deform_skirt_f_l [{shapeType=Capsule, length=1, radius=1, shapeName=deform_skirt_f_l_capsule}]
 deform_skirt_f_l_aim [{shapeType=Capsule, length=9.4594, radius=10, shapeName=deform_skirt_f_l_aim_capsule, center=<2.3481 -9.3764 0>}]
 deform_skirt_f_r [{shapeType=Capsule, length=1, radius=1, shapeName=deform_skirt_f_r_capsule}]
 deform_skirt_f_r_aim [{
   shapeType=Capsule
   length=11.5715
   radius=10
   shapeName=deform_skirt_f_r_aim_capsule
   center=<-2.5142 9.0435 2.2017>
   rotation=<rot -0.0001 -0.0002 -18.1847>
  }]
 deform_skirt_m_l [{shapeType=Capsule, length=1, radius=1, shapeName=deform_skirt_m_l_capsule}]
 deform_skirt_m_r [{shapeType=Capsule, length=1, radius=1, shapeName=deform_skirt_m_r_capsule}]
 deform_skirt_m_r_aim [{
   shapeType=Capsule
   length=10.3648
   radius=10
   shapeName=deform_skirt_m_r_aim_capsule
   center=<-3.0962 9.0435 -0.9>
   rotation=<rot 0 -0.0001 -5.8912>
  }]
 dyn_hoodie_1_l [{
   shapeType=Capsule
   length=13.616
   radius=4.9909
   shapeName=dyn_hoodie_1_l_capsule
   center=<10.1115 -3.9773 0>
   rotation=<rot 90 -179.4032 -173.2975>
  }]
 dyn_hoodie_1_r [
  {shapeType=Capsule, length=11.5878, radius=5.8449, shapeName=dyn_hoodie_1_r_capsule, center=<9.8899 -4.074 0>, rotation=<rot 90 0 0>}
  {
   shapeType=Capsule
   length=10.5372
   radius=7.8347
   shapeName=dyn_hoodie_1_r_capsule
   center=<9.9493 -4.9393 12.186>
   rotation=<rot -87.9257 -0.0002 0.0001>
  }
 ]
 dyn_hoodie_b_1_l [{shapeType=Capsule, length=14.4722, radius=6.5278, shapeName=dyn_hoodie_b_1_l_capsule, center=<-7.9624 4.7222 0>, rotation=<rot 90 0 0>}]
 dyn_hoodie_b_1_r [{shapeType=Capsule, length=14.4722, radius=4, shapeName=dyn_hoodie_b_1_l_capsule, center=<7.9624 -4.7222 0>, rotation=<rot 90 0 180>}]
 dyn_hoodie_b_l [{shapeType=Capsule, length=14.4722, radius=6.5278, shapeName=dyn_hoodie_b_l_capsule, center=<-10.7767 5.202 0>, rotation=<rot 90 0 0>}]
 dyn_hoodie_b_r [{shapeType=Capsule, length=14.4722, radius=4, shapeName=dyn_hoodie_b_l_capsule, center=<10.1606 -4.4403 0>, rotation=<rot 90 0 180>}]
 dyn_hoodie_bk_1_l [{shapeType=Capsule, length=12.2972, radius=8.7028, shapeName=dyn_hoodie_bk_1_l_capsule, center=<-11.6673 4.7222 0>}]
 dyn_hoodie_bk_1_r [{shapeType=Capsule, length=12.2972, radius=8.7028, shapeName=dyn_hoodie_bk_1_l_capsule, center=<11.6673 -4.7222 0>}]
 dyn_hoodie_bk_l [{shapeType=Capsule, length=14.4722, radius=6.5278, shapeName=dyn_hoodie_bk_l_capsule, center=<-10.7767 5.202 0>, rotation=<rot 90 0 0>}]
 dyn_hoodie_bk_r [{shapeType=Capsule, length=14.4722, radius=6.5278, shapeName=dyn_hoodie_bk_l_capsule, center=<10.1606 -4.4403 0>, rotation=<rot 90 0 180>}]
 dyn_hoodie_l [{shapeType=Capsule, length=22.3935, radius=5, shapeName=dyn_hoodie_l_capsule, center=<14.1688 -3.0118 0>, rotation=<rot 90 180 178.3206>}]
 dyn_hoodie_r [{shapeType=Capsule, length=24.8629, radius=6.3977, shapeName=dyn_hoodie_r_capsule, center=<14.5023 -4.3102 0>, rotation=<rot 90 180 177.6943>}]
 dyn_skirt_b_l [{shapeType=Capsule, length=20.6388, radius=9.3111, shapeName=dyn_skirt_b_l_capsule, center=<-10.9618 -9.4945 0>, rotation=<rot 90 0 0>}]
 dyn_skirt_b_r [{shapeType=Capsule, length=20.6388, radius=9.3111, shapeName=dyn_skirt_b_l_capsule, center=<10.9618 9.4945 0>, rotation=<rot 90 0 180>}]
 dyn_skirt_bk_l [{shapeType=Capsule, length=12.8187, radius=9.3111, shapeName=dyn_skirt_bk_l_capsule, center=<-14.9199 -6.4755 0>}]
 dyn_skirt_bk_r [{shapeType=Capsule, length=12.8187, radius=9.3111, shapeName=dyn_skirt_bk_l_capsule, center=<14.9199 6.4755 0>}]
 dyn_skirt_f_1_l [{shapeType=Capsule, length=1, radius=1, shapeName=dyn_skirt_f_1_l_capsule}]
 dyn_skirt_f_l [{shapeType=Capsule, length=1, radius=1, shapeName=dyn_skirt_f_l_capsule}]
 dyn_skirt_f_r [{shapeType=Capsule, length=21.8799, radius=10.8503, shapeName=dyn_skirt_f_r_capsule, center=<13.8208 9.3378 0>, rotation=<rot 90 0 0>}]
 dyn_skirt_f_r1 [{shapeType=Capsule, length=1, radius=1, shapeName=dyn_skirt_f_r1_capsule}]
 dyn_skirt_keychain_f_l [{
   shapeType=Capsule
   length=0.45
   radius=3.8459
   shapeName=dyn_skirt_keychain_f_l_capsule
   center=<5.5148 -0.0369 0>
   rotation=<rot 90 14.0362 -14.742>
  }]
 dyn_skirt_keychain_f_r [{shapeType=Capsule, length=2.9532, radius=2.0489, shapeName=dyn_skirt_keychain_f_r_capsule, center=<-2.8316 1.9827 0>, rotation=<rot 90 0 0>}]
 dyn_skirt_m_1_l [{shapeType=Capsule, length=7.7631, radius=9.6964, shapeName=dyn_skirt_m_1_l_capsule, center=<-7.9809 -8.2904 0>, rotation=<rot 90 0 0>}]
 dyn_skirt_m_l [{shapeType=Capsule, length=7.4043, radius=9.3111, shapeName=dyn_skirt_m_l_capsule, center=<-6.359 -8.319 0>, rotation=<rot 90 0 0>}]
 dyn_skirt_m_r [{shapeType=Capsule, length=7.4043, radius=9.3111, shapeName=dyn_skirt_m_l_capsule, center=<6.359 8.319 0>, rotation=<rot 90 0 180>}]
 dyn_skirt_m_r1 [{shapeType=Capsule, length=7.9387, radius=9.7889, shapeName=dyn_skirt_m_r1_capsule, center=<7.4665 10.2531 0>, rotation=<rot 90 0 0>}]
 pelvis [{shapeType=Capsule, length=6.5154, radius=12.2401, shapeName=pelvis_capsule, center=<0 -2.561 0>}]
 root [
  {shapeType=Sphere, radius=111.4502, shapeName=root_sphere, center=<0 0 -99.9427>}
  {shapeType=Capsule, length=0, radius=5, shapeName=root_capsule, rotation=<rot 90 0 0>}
 ]
 spine_01 [{shapeType=Capsule, length=10, radius=5, shapeName=spine_01_capsule, rotation=<rot 90 0 0>}]
 spine_02 [{shapeType=Capsule, length=10, radius=5, shapeName=spine_02_capsule, rotation=<rot 90 0 0>}]
 spine_03 [{shapeType=Capsule, length=5.0856, radius=9.9144, shapeName=spine_03_capsule}]
 spine_04 [{
   shapeType=Capsule
   length=4.5403
   radius=10.1288
   shapeName=spine_04_capsule
   center=<0.0055 2.1026 0.0225>
   rotation=<rot -0.0895 -90 0.4702>
  }]
 thigh_l [{shapeType=Capsule, length=38.3844, radius=7.2866, shapeName=thigh_l_capsule, center=<-22.2145 0 0>, rotation=<rot 90 0 0>}]
 thigh_r [{shapeType=Capsule, length=38.3844, radius=7.2866, shapeName=thigh_l_capsule, center=<22.2145 0 0>, rotation=<rot 90 0 180>}]
}
constraints [
 {
  bone1Name=spine_01
  bone2Name=pelvis
  swing1LimitDegrees=45
  swing1Motion=Free
  swing2LimitDegrees=45
  swing2Motion=Free
  twistLimitDegrees=45
  twistMotion=Free
 }
 {
  bone1Name=spine_02
  bone2Name=spine_01
  swing1LimitDegrees=45
  swing1Motion=Free
  swing2LimitDegrees=45
  swing2Motion=Free
  twistLimitDegrees=45
  twistMotion=Free
 }
 {
  bone1Name=spine_03
  bone2Name=spine_02
  swing1LimitDegrees=45
  swing1Motion=Free
  swing2LimitDegrees=45
  swing2Motion=Free
  twistLimitDegrees=45
  twistMotion=Free
 }
 {
  bone1Name=thigh_l
  bone2Name=pelvis
  swing1LimitDegrees=45
  swing1Motion=Free
  swing2LimitDegrees=45
  swing2Motion=Free
  twistLimitDegrees=45
  twistMotion=Free
 }
 {
  bone1Name=deform_hoodie_l
  bone2Name=spine_01
  swing1LimitDegrees=45
  swing1Motion=Free
  swing2LimitDegrees=45
  swing2Motion=Free
  twistLimitDegrees=45
  twistMotion=Free
 }
 {bone1Name=dyn_hoodie_l, bone2Name=deform_hoodie_l, swing1LimitDegrees=25, swing2LimitDegrees=25, twistLimitDegrees=1}
 {bone1Name=dyn_hoodie_1_l, bone2Name=dyn_hoodie_l, swing1LimitDegrees=15, swing2LimitDegrees=15, twistLimitDegrees=5}
 {
  bone1Name=deform_skirt_b_l
  bone2Name=spine_01
  swing1LimitDegrees=45
  swing1Motion=Free
  swing2LimitDegrees=45
  swing2Motion=Free
  twistLimitDegrees=45
  twistMotion=Free
 }
 {bone1Name=dyn_skirt_b_l, bone2Name=deform_skirt_b_l, swing1LimitDegrees=15, swing2LimitDegrees=10, twistLimitDegrees=1}
 {bone1Name=dyn_hoodie_b_l, bone2Name=dyn_skirt_b_l, swing1LimitDegrees=25, swing2LimitDegrees=25, twistLimitDegrees=5}
 {bone1Name=dyn_hoodie_b_1_l, bone2Name=dyn_hoodie_b_l, swing1LimitDegrees=25, swing2LimitDegrees=25, twistLimitDegrees=5}
 {
  bone1Name=deform_skirt_b_r
  bone2Name=spine_01
  swing1LimitDegrees=45
  swing1Motion=Free
  swing2LimitDegrees=45
  swing2Motion=Free
  twistLimitDegrees=45
  twistMotion=Free
 }
 {
  bone1Name=deform_skirt_bk_l
  bone2Name=spine_01
  swing1LimitDegrees=45
  swing1Motion=Free
  swing2LimitDegrees=45
  swing2Motion=Free
  twistLimitDegrees=45
  twistMotion=Free
 }
 {bone1Name=dyn_skirt_bk_l, bone2Name=deform_skirt_bk_l, swing1LimitDegrees=15, swing2LimitDegrees=10, twistLimitDegrees=1}
 {bone1Name=dyn_hoodie_bk_l, bone2Name=dyn_skirt_bk_l, swing1LimitDegrees=25, swing2LimitDegrees=25, twistLimitDegrees=5}
 {bone1Name=dyn_hoodie_bk_1_l, bone2Name=dyn_hoodie_bk_l, swing1LimitDegrees=25, swing2LimitDegrees=25, twistLimitDegrees=5}
 {
  bone1Name=deform_skirt_bk_r
  bone2Name=spine_01
  swing1LimitDegrees=45
  swing1Motion=Free
  swing2LimitDegrees=45
  swing2Motion=Free
  twistLimitDegrees=45
  twistMotion=Free
 }
 {
  bone1Name=deform_skirt_f_l
  bone2Name=spine_01
  swing1LimitDegrees=45
  swing1Motion=Free
  swing2LimitDegrees=45
  swing2Motion=Free
  twistLimitDegrees=45
  twistMotion=Free
 }
 {
  bone1Name=deform_skirt_chain_f_r
  bone2Name=dyn_skirt_f_l
  swing1LimitDegrees=45
  swing1Motion=Free
  swing2LimitDegrees=45
  swing2Motion=Free
  twistLimitDegrees=45
  twistMotion=Free
 }
 {bone1Name=dyn_skirt_keychain_f_r, bone2Name=dyn_skirt_f_l, swing1LimitDegrees=20, swing2LimitDegrees=15, twistLimitDegrees=1}
 {
  bone1Name=deform_skirt_f_r
  bone2Name=spine_01
  swing1LimitDegrees=45
  swing1Motion=Free
  swing2LimitDegrees=45
  swing2Motion=Free
  twistLimitDegrees=45
  twistMotion=Free
 }
 {
  bone1Name=deform_skirt_m_l
  bone2Name=spine_01
  swing1LimitDegrees=45
  swing1Motion=Free
  swing2LimitDegrees=45
  swing2Motion=Free
  twistLimitDegrees=45
  twistMotion=Free
 }
 {
  bone1Name=deform_skirt_chain_f_l
  bone2Name=deform_skirt_m_l
  swing1LimitDegrees=45
  swing1Motion=Free
  swing2LimitDegrees=45
  swing2Motion=Free
  twistLimitDegrees=45
  twistMotion=Free
 }
 {bone1Name=dyn_skirt_keychain_f_l, bone2Name=deform_skirt_chain_f_l, swing1LimitDegrees=20, swing2LimitDegrees=20, twistLimitDegrees=20}
 {bone1Name=dyn_skirt_m_l, bone2Name=deform_skirt_m_l, swing1LimitDegrees=15, swing2LimitDegrees=10, twistLimitDegrees=1}
 {bone1Name=dyn_skirt_m_1_l, bone2Name=dyn_skirt_m_l, swing1LimitDegrees=10, swing2LimitDegrees=20, twistLimitDegrees=1}
 {
  bone1Name=deform_skirt_m_r
  bone2Name=spine_01
  swing1LimitDegrees=45
  swing1Motion=Free
  swing2LimitDegrees=45
  swing2Motion=Free
  twistLimitDegrees=45
  twistMotion=Free
 }
 {bone1Name=dyn_skirt_m_r1, bone2Name=dyn_skirt_m_r, swing1LimitDegrees=15, swing2LimitDegrees=10, twistLimitDegrees=1}
 {
  bone1Name=pelvis
  bone2Name=root
  swing1LimitDegrees=45
  swing1Motion=Free
  swing2LimitDegrees=45
  swing2Motion=Free
  twistLimitDegrees=45
  twistMotion=Free
 }
 {
  bone1Name=thigh_r
  bone2Name=pelvis
  swing1LimitDegrees=45
  swing1Motion=Free
  swing2LimitDegrees=45
  swing2Motion=Free
  twistLimitDegrees=45
  twistMotion=Free
 }
 {bone1Name=dyn_skirt_m_r, bone2Name=deform_skirt_m_r, swing1LimitDegrees=15, swing2LimitDegrees=10, twistLimitDegrees=1}
 {bone1Name=dyn_skirt_b_r, bone2Name=deform_skirt_b_r, swing1LimitDegrees=15, swing2LimitDegrees=10, twistLimitDegrees=1}
 {bone1Name=dyn_skirt_bk_r, bone2Name=deform_skirt_bk_r, swing1LimitDegrees=15, swing2LimitDegrees=10, twistLimitDegrees=1}
 {bone1Name=dyn_hoodie_b_r, bone2Name=dyn_skirt_b_r, swing1LimitDegrees=25, swing2LimitDegrees=25, twistLimitDegrees=5}
 {bone1Name=dyn_hoodie_b_1_r, bone2Name=dyn_hoodie_b_r, swing1LimitDegrees=25, swing2LimitDegrees=25, twistLimitDegrees=5}
 {bone1Name=dyn_hoodie_bk_r, bone2Name=dyn_skirt_bk_r, swing1LimitDegrees=25, swing2LimitDegrees=25, twistLimitDegrees=5}
 {bone1Name=dyn_hoodie_bk_1_r, bone2Name=dyn_hoodie_bk_r, swing1LimitDegrees=25, swing2LimitDegrees=25, twistLimitDegrees=5}
 {
  bone1Name=deform_skirt_f_aim
  bone2Name=spine_01
  swing1LimitDegrees=45
  swing1Motion=Free
  swing2LimitDegrees=45
  swing2Motion=Free
  twistLimitDegrees=45
  twistMotion=Free
 }
 {
  bone1Name=deform_skirt_f_l_aim
  bone2Name=dyn_skirt_f_1_l
  swing1LimitDegrees=45
  swing1Motion=Free
  swing2LimitDegrees=45
  swing2Motion=Free
  twistLimitDegrees=45
  twistMotion=Free
 }
 {
  bone1Name=deform_skirt_f_r_aim
  bone2Name=dyn_skirt_f_r1
  swing1LimitDegrees=45
  swing1Motion=Free
  swing2LimitDegrees=45
  swing2Motion=Free
  twistLimitDegrees=45
  twistMotion=Free
 }
 {
  bone1Name=deform_skirt_m_r_aim
  bone2Name=dyn_skirt_m_r1
  swing1LimitDegrees=45
  swing1Motion=Free
  swing2LimitDegrees=45
  swing2Motion=Free
  twistLimitDegrees=45
  twistMotion=Free
 }
 {
  bone1Name=dyn_skirt_f_r
  bone2Name=deform_skirt_f_r
  swing1LimitDegrees=45
  swing1Motion=Free
  swing2LimitDegrees=45
  swing2Motion=Free
  twistLimitDegrees=45
  twistMotion=Free
 }
 {
  bone1Name=deform_hoodie_r
  bone2Name=dyn_skirt_f_r
  swing1LimitDegrees=45
  swing1Motion=Free
  swing2LimitDegrees=45
  swing2Motion=Free
  twistLimitDegrees=45
  twistMotion=Free
 }
 {bone1Name=dyn_hoodie_r, bone2Name=deform_hoodie_r, swing1LimitDegrees=35, swing2LimitDegrees=5, twistLimitDegrees=1}
 {bone1Name=dyn_hoodie_1_r, bone2Name=dyn_hoodie_r, swing1LimitDegrees=35, swing2LimitDegrees=15, twistLimitDegrees=10}
 {
  bone1Name=calf_l
  bone2Name=thigh_l
  swing1LimitDegrees=45
  swing1Motion=Free
  swing2LimitDegrees=45
  swing2Motion=Free
  twistLimitDegrees=45
  twistMotion=Free
 }
 {
  bone1Name=calf_r
  bone2Name=thigh_r
  swing1LimitDegrees=45
  swing1Motion=Free
  swing2LimitDegrees=45
  swing2Motion=Free
  twistLimitDegrees=45
  twistMotion=Free
 }
 {
  bone1Name=dyn_skirt_f_l
  bone2Name=deform_skirt_f_l
  swing1LimitDegrees=45
  swing1Motion=Free
  swing2LimitDegrees=45
  swing2Motion=Free
  twistLimitDegrees=45
  twistMotion=Free
 }
 {
  bone1Name=dyn_skirt_f_r1
  bone2Name=dyn_skirt_f_r
  swing1LimitDegrees=45
  swing1Motion=Free
  swing2LimitDegrees=45
  swing2Motion=Free
  twistLimitDegrees=45
  twistMotion=Free
 }
 {
  bone1Name=dyn_skirt_f_1_l
  bone2Name=dyn_skirt_f_l
  swing1LimitDegrees=45
  swing1Motion=Free
  swing2LimitDegrees=45
  swing2Motion=Free
  twistLimitDegrees=45
  twistMotion=Free
 }
 {
  bone1Name=dyn_hoodie_bk_1_r
  bone2Name=dyn_hoodie_b_1_r
  swing1LimitDegrees=20
  swing1Motion=Free
  swing2LimitDegrees=20
  swing2Motion=Free
  twistLimitDegrees=10
  twistMotion=Free
 }
 {
  bone1Name=dyn_hoodie_bk_1_l
  bone2Name=dyn_hoodie_b_1_l
  swing1LimitDegrees=20
  swing1Motion=Free
  swing2LimitDegrees=20
  swing2Motion=Free
  twistLimitDegrees=10
  twistMotion=Free
 }
 {
  bone1Name=dyn_hoodie_bk_1_r
  bone2Name=dyn_hoodie_bk_1_l
  swing1LimitDegrees=20
  swing1Motion=Free
  swing2LimitDegrees=20
  swing2Motion=Free
  twistLimitDegrees=10
  twistMotion=Free
 }
 {
  bone1Name=spine_04
  bone2Name=spine_03
  swing1LimitDegrees=45
  swing1Motion=Free
  swing2LimitDegrees=45
  swing2Motion=Free
  twistLimitDegrees=45
  twistMotion=Free
 }
]
