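# Physics asset for the upper body of the F_MED_PocketScrunchie female cosmetic body, covering its hoodie, phone, tie and front skirt bones.
# Capsules on pelvis, spine, neck, clavicles, upper arms and right pec form the body, mostly joined by angularly free constraints.
# Phone and tie chains hang from spine_05 with limited swing and twist, using boxes for the tie.
# Cross-constraints link left and right phone chains, phone to tie, and right hoodie to skirt.
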PhysicsAsset /BRCosmetics/Characters/Player/Female/Medium/Bodies/F_MED_PocketScrunchie/Meshes/F_MED_PocketScrunchie_Physics_Upper
bodies {
 clavicle_l [{shapeType=Capsule, length=4.973, radius=6.0597, shapeName=clavicle_l_capsule, center=<5.6719 0 -0.7429>, rotation=<rot 90 0 0>}]
 clavicle_r [{shapeType=Capsule, length=4.973, radius=6.0597, shapeName=clavicle_l_capsule, center=<-5.6719 0 0.7429>, rotation=<rot 90 0 180>}]
 deform_hoodie_l [{shapeType=Capsule, length=1, radius=1, shapeName=deform_hoodie_l_capsule, rotation=<rot 90 0 0>}]
 deform_pec_r [{shapeType=Capsule, length=5, radius=5, shapeName=deform_pec_r_capsule, center=<0 0.7127 0>, rotation=<rot 58.2386 -0.0002 -0.0002>}]
 deform_phone_l [{shapeType=Capsule, length=1, radius=1, shapeName=deform_phone_l_capsule, rotation=<rot 90 0 0>}]
 deform_phone_r [{shapeType=Capsule, length=1, radius=1, shapeName=deform_phone_r_capsule, rotation=<rot 90 0 0>}]
 deform_tie [{shapeType=Capsule, length=1, radius=1, shapeName=deform_tie_capsule, rotation=<rot 90 0 0>}]
 dyn_hoodie_l [{shapeType=Capsule, length=16.0764, radius=7.9448, shapeName=dyn_hoodie_l_capsule, center=<11.4068 6.8663 0>, rotation=<rot 90 0 0>}]
 dyn_hoodie_r [{
   shapeType=Capsule
   length=14.3967
   radius=8.0936
   shapeName=dyn_hoodie_r_capsule
   center=<11.8297 7.0022 0.3381>
   rotation=<rot 1.2962 -90 92.4781>
  }]
 dyn_phone_1_l [{shapeType=Capsule, length=5.2874, radius=2.1926, shapeName=dyn_phone_1_l_capsule, center=<4.1638 -0.9673 0>, rotation=<rot 90 0 0>}]
 dyn_phone_1_r [{
   shapeType=Capsule
   length=7.6288
   radius=1.6455
   shapeName=dyn_phone_1_r_capsule
   center=<4.5668 -0.9646 0>
   rotation=<rot 90 179.5524 176.742>
  }]
 dyn_phone_2_r [{shapeType=Capsule, length=2.3668, radius=3.848, shapeName=dyn_phone_2_r_capsule, center=<2.6112 -4.6155 1.2479>, rotation=<rot 90 0 0>}]
 dyn_phone_l [{shapeType=Capsule, length=6, radius=1.48, shapeName=dyn_phone_l_capsule, center=<4.1638 -0.9673 0>, rotation=<rot 90 0 0>}]
 dyn_phone_r [{shapeType=Capsule, length=6.0169, radius=1.4631, shapeName=dyn_phone_r_capsule, center=<4.1613 -1.5721 0>, rotation=<rot 90 180 177.2637>}]
 dyn_skirt_f_l [{shapeType=Capsule, length=4.8783, radius=10.1217, shapeName=dyn_skirt_f_r_capsule, center=<-10.745 -5.9226 0>, rotation=<rot 0 0 -180>}]
 dyn_skirt_f_r [{shapeType=Capsule, length=4.8783, radius=10.1217, shapeName=dyn_skirt_f_r_capsule, center=<10.745 5.9226 0>}]
 dyn_tie [{shapeType=Box, extentX=7.62, extentY=1.1259, extentZ=12.7615, shapeName=dyn_tie_box, center=<3.6527 -0.38 0>}]
 dyn_tie_1 [{shapeType=Box, extentX=10.3012, extentY=1.3247, extentZ=6.6558, shapeName=dyn_tie_1_box, center=<5.5641 -0.4301 0>}]
 dyn_tie_2 [{
   shapeType=Box
   extentX=11.9719
   extentY=1.9799
   extentZ=6.0184
   shapeName=dyn_tie_2_box
   center=<5.3035 2.5695 0>
   rotation=<rot 0 22.1486 0>
  }]
 neck_01 [{shapeType=Capsule, length=10, radius=5, shapeName=neck_01_capsule, rotation=<rot 90 0 0>}]
 pelvis [{shapeType=Capsule, length=6.5154, radius=12.2401, shapeName=pelvis_capsule, center=<0 -2.561 0>}]
 root [{shapeType=Capsule, length=0, radius=5, shapeName=root_capsule, rotation=<rot 90 0 0>}]
 spine_01 [{shapeType=Capsule, length=11.582, radius=8.7733, shapeName=spine_01_capsule, center=<0 1.2553 0>}]
 spine_02 [{shapeType=Capsule, length=14.7972, radius=7.9376, shapeName=spine_02_capsule, center=<0 2.6015 0>}]
 spine_03 [{shapeType=Capsule, length=12.7348, radius=10, shapeName=spine_03_capsule, center=<0 1.7133 0>}]
 spine_04 [{shapeType=Capsule, length=12.7348, radius=10, shapeName=spine_04_capsule, center=<0 3.1839 0>}]
 spine_05 [{shapeType=Capsule, length=12.7348, radius=10, shapeName=spine_05_capsule, center=<-1.7223 -0.5295 0>}]
 upperarm_l [{shapeType=Capsule, length=17.5068, radius=5, shapeName=upperarm_l_capsule, center=<7.6366 0 0>, rotation=<rot 90 0 0>}]
 upperarm_r [{shapeType=Capsule, length=17.5068, radius=5, shapeName=upperarm_l_capsule, center=<-7.6366 0 0>, rotation=<rot 90 0 180>}]
 upperarm_twist_01_l [{shapeType=Capsule, length=17.4262, radius=6.005, shapeName=upperarm_twist_01_l_capsule, center=<3.6149 0 0>, rotation=<rot 90 0 0>}]
 upperarm_twist_01_r [{shapeType=Capsule, length=17.4262, radius=6.005, shapeName=upperarm_twist_01_l_capsule, center=<-3.6149 0 0>, rotation=<rot 90 0 180>}]
}
constraints [
 {
  bone1Name=spine_01
  bone2Name=pelvis
  swing1LimitDegrees=45
  swing1Motion=Free
  swing2LimitDegrees=45
  swing2Motion=Free
  twistLimitDegrees=45
  twistMotion=Free
 }
 {
  bone1Name=pelvis
  bone2Name=root
  swing1LimitDegrees=45
  swing1Motion=Free
  swing2LimitDegrees=45
  swing2Motion=Free
  twistLimitDegrees=45
  twistMotion=Free
 }
 {
  bone1Name=spine_05
  bone2Name=spine_04
  swing1LimitDegrees=45
  swing1Motion=Free
  swing2LimitDegrees=45
  swing2Motion=Free
  twistLimitDegrees=45
  twistMotion=Free
 }
 {
  bone1Name=clavicle_l
  bone2Name=spine_05
  swing1LimitDegrees=45
  swing1Motion=Free
  swing2LimitDegrees=45
  swing2Motion=Free
  twistLimitDegrees=45
  twistMotion=Free
 }
 {
  bone1Name=upperarm_twist_01_l
  bone2Name=upperarm_l
  swing1LimitDegrees=45
  swing1Motion=Free
  swing2LimitDegrees=45
  swing2Motion=Free
  twistLimitDegrees=45
  twistMotion=Free
 }
 {
  bone1Name=upperarm_l
  bone2Name=clavicle_l
  swing1LimitDegrees=45
  swing1Motion=Free
  swing2LimitDegrees=45
  swing2Motion=Free
  twistLimitDegrees=45
  twistMotion=Free
 }
 {
  bone1Name=neck_01
  bone2Name=spine_05
  swing1LimitDegrees=45
  swing1Motion=Free
  swing2LimitDegrees=45
  swing2Motion=Free
  twistLimitDegrees=45
  twistMotion=Free
 }
 {
  bone1Name=spine_02
  bone2Name=spine_01
  swing1LimitDegrees=45
  swing1Motion=Free
  swing2LimitDegrees=45
  swing2Motion=Free
  twistLimitDegrees=45
  twistMotion=Free
 }
 {
  bone1Name=spine_03
  bone2Name=spine_02
  swing1LimitDegrees=45
  swing1Motion=Free
  swing2LimitDegrees=45
  swing2Motion=Free
  twistLimitDegrees=45
  twistMotion=Free
 }
 {
  bone1Name=spine_04
  bone2Name=spine_03
  swing1LimitDegrees=45
  swing1Motion=Free
  swing2LimitDegrees=45
  swing2Motion=Free
  twistLimitDegrees=45
  twistMotion=Free
 }
 {
  bone1Name=deform_hoodie_l
  bone2Name=spine_01
  swing1LimitDegrees=45
  swing1Motion=Free
  swing2LimitDegrees=45
  swing2Motion=Free
  twistLimitDegrees=45
  twistMotion=Free
 }
 {
  bone1Name=dyn_hoodie_l
  bone2Name=deform_hoodie_l
  swing1LimitDegrees=45
  swing1Motion=Free
  swing2LimitDegrees=45
  swing2Motion=Free
  twistLimitDegrees=45
  twistMotion=Free
 }
 {
  bone1Name=upperarm_twist_01_r
  bone2Name=upperarm_r
  swing1LimitDegrees=45
  swing1Motion=Free
  swing2LimitDegrees=45
  swing2Motion=Free
  twistLimitDegrees=45
  twistMotion=Free
 }
 {
  bone1Name=upperarm_r
  bone2Name=clavicle_r
  swing1LimitDegrees=45
  swing1Motion=Free
  swing2LimitDegrees=45
  swing2Motion=Free
  twistLimitDegrees=45
  twistMotion=Free
 }
 {
  bone1Name=clavicle_r
  bone2Name=spine_05
  swing1LimitDegrees=45
  swing1Motion=Free
  swing2LimitDegrees=45
  swing2Motion=Free
  twistLimitDegrees=45
  twistMotion=Free
 }
 {
  bone1Name=deform_phone_l
  bone2Name=spine_05
  swing1LimitDegrees=45
  swing1Motion=Free
  swing2LimitDegrees=45
  swing2Motion=Free
  twistLimitDegrees=45
  twistMotion=Free
 }
 {bone1Name=dyn_phone_l, bone2Name=deform_phone_l, swing1LimitDegrees=30, swing2LimitDegrees=15, twistLimitDegrees=1}
 {bone1Name=dyn_phone_1_l, bone2Name=dyn_phone_l, swing1LimitDegrees=30, swing2LimitDegrees=15, twistLimitDegrees=1}
 {
  bone1Name=deform_phone_r
  bone2Name=spine_05
  swing1LimitDegrees=45
  swing1Motion=Free
  swing2LimitDegrees=45
  swing2Motion=Free
  twistLimitDegrees=45
  twistMotion=Free
 }
 {bone1Name=dyn_phone_r, bone2Name=deform_phone_r, swing1LimitDegrees=30, swing2LimitDegrees=15, twistLimitDegrees=1}
 {bone1Name=dyn_phone_1_r, bone2Name=dyn_phone_r, swing1LimitDegrees=30, swing2LimitDegrees=15, twistLimitDegrees=1}
 {bone1Name=dyn_phone_2_r, bone2Name=dyn_phone_1_r, swing1LimitDegrees=45, swing2LimitDegrees=45, twistLimitDegrees=5}
 {
  bone1Name=deform_tie
  bone2Name=spine_05
  swing1LimitDegrees=45
  swing1Motion=Free
  swing2LimitDegrees=45
  swing2Motion=Free
  twistLimitDegrees=45
  twistMotion=Free
 }
 {bone1Name=dyn_phone_1_l, bone2Name=dyn_phone_1_r, swing1LimitDegrees=5, swing2LimitDegrees=5, twistLimitDegrees=5}
 {
  bone1Name=dyn_skirt_f_r
  bone2Name=spine_01
  swing1LimitDegrees=45
  swing1Motion=Free
  swing2LimitDegrees=45
  swing2Motion=Free
  twistLimitDegrees=45
  twistMotion=Free
 }
 {
  bone1Name=dyn_skirt_f_l
  bone2Name=spine_01
  swing1LimitDegrees=45
  swing1Motion=Free
  swing2LimitDegrees=45
  swing2Motion=Free
  twistLimitDegrees=45
  twistMotion=Free
 }
 {
  bone1Name=dyn_phone_1_l
  bone2Name=dyn_phone_2_r
  swing1LimitDegrees=45
  swing1Motion=Free
  swing2LimitDegrees=45
  swing2Motion=Free
  twistLimitDegrees=45
  twistMotion=Free
 }
 {bone1Name=dyn_tie, bone2Name=deform_tie, swing1LimitDegrees=35, swing2LimitDegrees=10, twistLimitDegrees=5}
 {bone1Name=dyn_tie_1, bone2Name=dyn_tie, swing1LimitDegrees=35, swing2LimitDegrees=10, twistLimitDegrees=5}
 {bone1Name=dyn_tie_2, bone2Name=dyn_tie_1, swing1LimitDegrees=35, swing2LimitDegrees=10, twistLimitDegrees=5}
 {bone1Name=dyn_phone_l, bone2Name=dyn_tie, swing1LimitDegrees=35, swing2LimitDegrees=25, twistLimitDegrees=45, twistMotion=Free}
 {
  bone1Name=dyn_phone_1_l
  bone2Name=dyn_tie
  swing1LimitDegrees=35
  swing2LimitDegrees=45
  swing2Motion=Free
  twistLimitDegrees=45
  twistMotion=Free
 }
 {
  bone1Name=dyn_phone_r
  bone2Name=dyn_tie
  swing1LimitDegrees=45
  swing2LimitDegrees=45
  swing2Motion=Free
  twistLimitDegrees=45
  twistMotion=Free
 }
 {
  bone1Name=dyn_hoodie_r
  bone2Name=dyn_skirt_f_r
  swing1LimitDegrees=45
  swing1Motion=Free
  swing2LimitDegrees=45
  swing2Motion=Free
  twistLimitDegrees=45
  twistMotion=Free
 }
 {
  bone1Name=deform_pec_r
  bone2Name=spine_04
  swing1LimitDegrees=45
  swing1Motion=Free
  swing2LimitDegrees=45
  swing2Motion=Free
  twistLimitDegrees=45
  twistMotion=Free
 }
 {
  bone1Name=dyn_tie_2
  bone2Name=dyn_phone_2_r
  swing1LimitDegrees=45
  swing1Motion=Free
  swing2LimitDegrees=45
  swing2Motion=Free
  twistLimitDegrees=45
  twistMotion=Free
 }
]
