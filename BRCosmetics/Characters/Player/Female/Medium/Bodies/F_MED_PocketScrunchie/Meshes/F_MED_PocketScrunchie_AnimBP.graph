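# Animation Blueprint for the PocketScrunchie female medium cosmetic body, adding secondary motion to the pose copied from its attached parent mesh.
# A Control Rig taking knee-slide and skydive inputs feeds skirt AnimDynamics, a post Control Rig, two spine-relative RigidBody simulations, a phone-bone LookAt, then lace, keychain and hoodie AnimDynamics.
# The event graph inverts sliding into an alpha, fading out right-skirt dynamics and both RigidBody nodes during slides.
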AnimBlueprint /BRCosmetics/Characters/Player/Female/Medium/Bodies/F_MED_PocketScrunchie/Meshes/F_MED_PocketScrunchie_AnimBP
target_skeleton=&/F_MED_PocketScrunchie_Skeleton
graphs [
 AnimGraph {nodes [
   n1 = Control Rig!(
    Source=Copy Pose From Mesh() {bUseAttachedParent=true}
    Alpha
    IsKneeSliding:float=Get Is KneeSliding() -> (Is KneeSliding:double)
    Skydiving_Alpha:bool=Get bIsSkyDiving()
   ) -> (Pose) {
    alphaBoolBlend {blendOption=Linear}
    controlRigAssetReference {blueprintRigClass=&/F_MED_PocketScrunchie_CtrlRig.F_MED_PocketScrunchie_CtrlRig_C}
   }
   n2 = Get Is Sliding Alpha() -> (Is Sliding Alpha:double)
   n3 = Anim Dynamics - Bone: dyn_skirt_f_r!(ComponentPose=Local To Component(LocalPose=@n1), Alpha=@n2, GravityOverride:Vector="1,1,1") -> (Pose) {
    angularDampingOverride=0.9999
    angularSpringConstant=1000
    bAngularSpring=true
    bGravityOverrideInSimSpace=true
    bOverrideAngularDamping=true
    bOverrideLinearDamping=true
    bUseGravityOverride=true
    linearDampingOverride=1
    simulationSpace=RootRelative
    alphaBoolBlend {blendOption=Linear}
    alphaScaleBiasClamp {bInterpResult=true}
    boundBone {boneName=dyn_skirt_f_r}
    componentAppliedLinearAccClamp=<200 200 200>
    componentLinearAccScale=<0.15 0.15 0.15>
    componentLinearVelScale=<0.15 0.15 0.15>
    gravityOverride=<1 1 1>
    physicsBodyDefinitions [{
      sphereCollisionRadius=0
      boundBone {boneName=dyn_skirt_f_r}
      boxExtents=<20 5 10>
      constraintSetup {angularLimitsMax=<0 5 -5>, angularLimitsMin=<0 -5 -60>, angularTarget=<1 0 0>}
      localJointOffset=<2 -0.1 0>
     }]
    relativeSpaceBone {boneName=root}
   }
   n4 = Anim Dynamics - Bone: dyn_skirt_f_r1!(ComponentPose=@n3, Alpha=@n2, GravityOverride:Vector="1,1,1") -> (Pose) {
    angularDampingOverride=0.9999
    angularSpringConstant=1000
    bAngularSpring=true
    bGravityOverrideInSimSpace=true
    bOverrideAngularDamping=true
    bOverrideLinearDamping=true
    bUseGravityOverride=true
    linearDampingOverride=1
    simulationSpace=RootRelative
    alphaBoolBlend {blendOption=Linear}
    alphaScaleBiasClamp {bInterpResult=true}
    boundBone {boneName=dyn_skirt_f_r1}
    componentAppliedLinearAccClamp=<200 200 200>
    componentLinearAccScale=<0.15 0.15 0.15>
    componentLinearVelScale=<0.15 0.15 0.15>
    gravityOverride=<1 1 1>
    physicsBodyDefinitions [{
      sphereCollisionRadius=0
      boundBone {boneName=dyn_skirt_f_r1}
      boxExtents=<20 5 10>
      constraintSetup {angularLimitsMax=<1 20 -1>, angularLimitsMin=<-1 -20 -60>, angularTarget=<1 0 0>}
      localJointOffset=<2 -0.1 0>
     }]
    relativeSpaceBone {boneName=root}
   }
   n5 = Anim Dynamics - Bone: dyn_skirt_f_l!(ComponentPose=@n4, Alpha=1, GravityOverride:Vector="1,1,1") -> (Pose) {
    angularDampingOverride=0.9999
    angularSpringConstant=1000
    bAngularSpring=true
    bGravityOverrideInSimSpace=true
    bOverrideAngularDamping=true
    bOverrideLinearDamping=true
    bUseGravityOverride=true
    linearDampingOverride=1
    simulationSpace=RootRelative
    alphaBoolBlend {blendOption=Linear}
    boundBone {boneName=dyn_skirt_f_l}
    componentAppliedLinearAccClamp=<200 200 200>
    componentLinearAccScale=<0.15 0.15 0.15>
    componentLinearVelScale=<0.15 0.15 0.15>
    gravityOverride=<1 1 1>
    physicsBodyDefinitions [{
      sphereCollisionRadius=0
      boundBone {boneName=dyn_skirt_f_l}
      boxExtents=<20 5 10>
      constraintSetup {angularLimitsMax=<0 5 -5>, angularLimitsMin=<0 -5 -30>, angularTarget=<1 0 0>}
      localJointOffset=<-2 0.1 0>
     }]
    relativeSpaceBone {boneName=root}
   }
   n6 = Anim Dynamics - Bone: dyn_skirt_f_1_l!(ComponentPose=@n5, Alpha=1, GravityOverride:Vector="1,1,1") -> (Pose) {
    angularDampingOverride=0.9999
    angularSpringConstant=1000
    bAngularSpring=true
    bGravityOverrideInSimSpace=true
    bOverrideAngularDamping=true
    bOverrideLinearDamping=true
    bUseGravityOverride=true
    linearDampingOverride=1
    simulationSpace=RootRelative
    alphaBoolBlend {blendOption=Linear}
    boundBone {boneName=dyn_skirt_f_1_l}
    componentAppliedLinearAccClamp=<200 200 200>
    componentLinearAccScale=<0.15 0.15 0.15>
    componentLinearVelScale=<0.15 0.15 0.15>
    gravityOverride=<1 1 1>
    physicsBodyDefinitions [{
      sphereCollisionRadius=0
      boundBone {boneName=dyn_skirt_f_1_l}
      boxExtents=<20 5 10>
      constraintSetup {angularLimitsMax=<0 20 -1>, angularLimitsMin=<0 -20 -60>, angularTarget=<1 0 0>}
      localJointOffset=<-1 0.1 0>
     }]
    relativeSpaceBone {boneName=root}
   }
   n7 = Control Rig!(Source=Component To Local(ComponentPose=@n6), Alpha) -> (Pose) {
    alphaBoolBlend {blendOption=Linear}
    controlRigAssetReference {blueprintRigClass=&/F_MED_PocketScrunchie_CtrlRig_Post.F_MED_PocketScrunchie_CtrlRig_Post_C}
   }
   n8 = Get Is Sliding Alpha()
   n9 = RigidBody!(ComponentPose=Local To Component(LocalPose=@n7), Alpha=@n8, OverrideWorldGravity:Vector="0,0,0", ExternalForce) -> (Pose) {
    bClampLinearTranslationLimitToRefPose=true
    bOverrideWorldGravity=true
    overridePhysicsAsset=&/F_MED_PocketScrunchie_Physics
    simulationSpace=BaseBoneSpace
    alphaBoolBlend {blendOption=Linear}
    alphaScaleBiasClamp {bInterpResult=true, interpSpeedDecreasing=5, interpSpeedIncreasing=5}
    baseBoneRef {boneName=spine_01}
    componentLinearAccScale=<0.4 0.4 0.4>
    componentLinearVelScale=<0.6 0.6 0.6>
    overrideWorldGravity {z=-980}
    simSpaceSettings {worldAlpha=0.4}
   }
   n10 = RigidBody(ComponentPose=@n9, Alpha=@n8, OverrideWorldGravity="0,0,0") {
    bClampLinearTranslationLimitToRefPose=true
    bOverrideWorldGravity=true
    overridePhysicsAsset=&/F_MED_PocketScrunchie_Physics_Upper
    simulationSpace=BaseBoneSpace
    alphaBoolBlend {blendOption=Linear}
    alphaScaleBiasClamp {bInterpResult=true, interpSpeedDecreasing=5, interpSpeedIncreasing=5}
    baseBoneRef {boneName=spine_05}
    componentAppliedLinearAccClamp=<3000 3000 3000>
    componentLinearAccScale=<0.4 0.4 0.4>
    componentLinearVelScale=<0.6 0.6 0.6>
    overrideWorldGravity {z=-980}
    simSpaceSettings {maxAngularAcceleration=1500, maxLinearAcceleration=1500, worldAlpha=0.5}
   }
   n11 = Look At - Bone: dyn_phone_1_l(ComponentPose=@n10, Alpha=1) {
    bUseLookUpAxis=true
    alphaBoolBlend {blendOption=Linear}
    boneToModify {boneName=dyn_phone_1_l}
    lookAtLocation {x=0}
    lookAtTarget {boneReference {boneName=dyn_phone_2_r}}
    lookAt_Axis {axis=<1 0>}
    lookUp_Axis {axis=<0 -1>}
   }
   n12 = Anim Dynamics - Bone: dyn_laces_l!(ComponentPose=@n11, Alpha=1, GravityOverride:Vector="1,1,1") -> (Pose) {
    angularDampingOverride=0.9999
    angularSpringConstant=1000
    bAngularSpring=true
    bGravityOverrideInSimSpace=true
    bOverrideAngularDamping=true
    bOverrideLinearDamping=true
    bUseGravityOverride=true
    linearDampingOverride=1
    simulationSpace=RootRelative
    alphaBoolBlend {blendOption=Linear}
    boundBone {boneName=dyn_laces_l}
    componentAppliedLinearAccClamp=<200 200 200>
    componentLinearAccScale=<0.15 0.15 0.15>
    componentLinearVelScale=<0.15 0.15 0.15>
    gravityOverride=<1 1 1>
    physicsBodyDefinitions [{
      sphereCollisionRadius=0
      boundBone {boneName=dyn_laces_l}
      boxExtents=<20 5 10>
      constraintSetup {angularLimitsMax=<1 -3 30>, angularLimitsMin=<-1 -20 -30>, angularTarget=<1 0 0>}
      localJointOffset=<-2 0.1 0>
     }]
    relativeSpaceBone {boneName=root}
   }
   n13 = Anim Dynamics - Bone: dyn_laces_r!(ComponentPose=@n12, Alpha=1, GravityOverride:Vector="1,1,1") -> (Pose) {
    angularDampingOverride=0.9999
    angularSpringConstant=1000
    bAngularSpring=true
    bGravityOverrideInSimSpace=true
    bOverrideAngularDamping=true
    bOverrideLinearDamping=true
    bUseGravityOverride=true
    linearDampingOverride=1
    simulationSpace=RootRelative
    alphaBoolBlend {blendOption=Linear}
    boundBone {boneName=dyn_laces_r}
    componentAppliedLinearAccClamp=<200 200 200>
    componentLinearAccScale=<0.15 0.15 0.15>
    componentLinearVelScale=<0.15 0.15 0.15>
    gravityOverride=<1 1 1>
    physicsBodyDefinitions [{
      sphereCollisionRadius=0
      boundBone {boneName=dyn_laces_r}
      boxExtents=<20 5 10>
      constraintSetup {angularLimitsMax=<1 -6 30>, angularLimitsMin=<-1 -20 -30>, angularTarget=<1 0 0>}
      localJointOffset=<2 -0.1 0>
     }]
    relativeSpaceBone {boneName=root}
   }
   n14 = Anim Dynamics - Bone: dyn_skirt_keychain_f_r!(ComponentPose=@n13, Alpha=1, GravityOverride:Vector="1,1,1") -> (Pose) {
    angularDampingOverride=0.9999
    angularSpringConstant=1000
    bAngularSpring=true
    bGravityOverrideInSimSpace=true
    bOverrideAngularDamping=true
    bOverrideLinearDamping=true
    bUseGravityOverride=true
    linearDampingOverride=1
    simulationSpace=RootRelative
    alphaBoolBlend {blendOption=Linear}
    boundBone {boneName=dyn_skirt_keychain_f_r}
    componentAppliedLinearAccClamp=<200 200 200>
    componentLinearAccScale=<0.15 0.15 0.15>
    componentLinearVelScale=<0.15 0.15 0.15>
    gravityOverride=<1 1 1>
    physicsBodyDefinitions [{
      sphereCollisionRadius=0
      boundBone {boneName=dyn_skirt_keychain_f_r}
      boxExtents=<20 5 10>
      constraintSetup {angularLimitsMax=<5 0 -6>, angularLimitsMin=<-1 -20 -30>, angularTarget=<1 0 0>}
      localJointOffset=<-1 0 0>
     }]
    relativeSpaceBone {boneName=root}
   }
   n15 = Anim Dynamics - Bone: dyn_hoodie_r!(ComponentPose=@n14, Alpha=1, GravityOverride:Vector="1,1,1") -> (Pose) {
    angularDampingOverride=0.9999
    angularSpringConstant=500
    bAngularSpring=true
    bOverrideAngularDamping=true
    bOverrideLinearDamping=true
    bUseGravityOverride=true
    linearDampingOverride=1
    simulationSpace=RootRelative
    alphaBoolBlend {blendOption=Linear}
    boundBone {boneName=dyn_hoodie_r}
    componentAppliedLinearAccClamp=<300 300 300>
    componentLinearAccScale=<0.15 0.15 0.15>
    componentLinearVelScale=<0.15 0.15 0.15>
    gravityOverride=<1 1 1>
    physicsBodyDefinitions [{
      sphereCollisionRadius=0
      boundBone {boneName=dyn_hoodie_r}
      boxExtents=<20 5 10>
      constraintSetup {angularLimitsMax=<1 0 0>, angularLimitsMin=<-1 0 -30>, angularTarget=<1 0 0>}
      localJointOffset=<1 0 0>
     }]
    relativeSpaceBone {boneName=root}
   }
   n16 = Anim Dynamics - Bone: dyn_hoodie_1_r!(ComponentPose=@n15, Alpha=1, GravityOverride:Vector="1,1,1") -> (Pose) {
    angularDampingOverride=0.9999
    angularSpringConstant=500
    bAngularSpring=true
    bOverrideAngularDamping=true
    bOverrideLinearDamping=true
    bUseGravityOverride=true
    linearDampingOverride=1
    simulationSpace=RootRelative
    alphaBoolBlend {blendOption=Linear}
    boundBone {boneName=dyn_hoodie_1_r}
    componentAppliedLinearAccClamp=<300 300 300>
    componentLinearAccScale=<0.15 0.15 0.15>
    componentLinearVelScale=<0.15 0.15 0.15>
    gravityOverride=<1 1 1>
    physicsBodyDefinitions [{
      sphereCollisionRadius=0
      boundBone {boneName=dyn_hoodie_1_r}
      boxExtents=<20 5 10>
      constraintSetup {angularLimitsMax=<1 10 0>, angularLimitsMin=<-1 -10 -30>, angularTarget=<1 0 0>}
      localJointOffset=<5 0 0>
     }]
    relativeSpaceBone {boneName=root}
   }
   Output Pose(Result=Component To Local(ComponentPose=@n16)) {layerGroup=None}
  ]}
 EventGraph {nodes [
   Event BlueprintUpdateAnimation()
   > n1 = Gravity Override State Function DEPRECATED(CharacterPartAnimInstance=Self-Reference(), GameTimeInSeconds=GetGameTimeInSeconds())
   > Set Gravity Override(Gravity Override=@n1)
   > Set LOD is Greater 1(LOD is Greater 1:bool="integer >= integer"(A=Get CurrentLOD(), B=2)) -> (Output_Get:bool)
   n2 = Get bIsSliding()
   > Set Is Sliding Alpha(Is Sliding Alpha:double="To Float (Boolean)"(InBool=NOT Boolean(A=@n2))) -> (Output_Get:double)
   > Set Is KneeSliding(Is KneeSliding:double="To Float (Boolean)"(InBool=Get bIsKneeSliding())) -> (Output_Get:double)
  ]}
]
variables [
 Gravity Override {type=Vector}
 Is KneeSliding {type=number}
 Is Sliding Alpha {type=number}
 LOD is Greater 1 {type=boolean}
]
