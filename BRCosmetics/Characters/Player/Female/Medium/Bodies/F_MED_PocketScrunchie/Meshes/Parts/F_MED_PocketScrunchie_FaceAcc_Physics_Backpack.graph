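# Physics asset for the F_MED_PocketScrunchie female character's FaceAcc part, Backpack variant, covering dynamic hair and two tails.
# Head, neck, upper spine, clavicle and upperarm bodies use free constraints, with a large offset spine_05 capsule likely representing the backpack.
# Deform root bodies join the head freely and lead into angle-limited dyn chains, stiffest at tail bases.
# Side hair chains fork, with dyn_hair_3 attached to the base dyn_hair body.
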PhysicsAsset /BRCosmetics/Characters/Player/Female/Medium/Bodies/F_MED_PocketScrunchie/Meshes/Parts/F_MED_PocketScrunchie_FaceAcc_Physics_Backpack
bodies {
 clavicle_l [{shapeType=Capsule, length=6.8911, radius=6.2588, shapeName=clavicle_l_capsule, center=<5.6238 -1.094 -1.0127>, rotation=<rot 90 0 0>}]
 clavicle_r [{shapeType=Capsule, length=6.8911, radius=6.2588, shapeName=clavicle_l_capsule, center=<-5.6238 1.094 1.0127>, rotation=<rot 90 0 180>}]
 deform_hair_f [{shapeType=Capsule, length=1, radius=1, shapeName=deform_hair_f_capsule, rotation=<rot 90 0 0>}]
 deform_hair_l [{shapeType=Capsule, length=1, radius=1, shapeName=deform_hair_l_capsule, rotation=<rot 90 0 0>}]
 deform_hair_r [{shapeType=Capsule, length=1, radius=1, shapeName=deform_hair_r_capsule, rotation=<rot 90 0 0>}]
 deform_tail_l [{shapeType=Capsule, length=5, radius=3, shapeName=deform_tail_l_capsule, rotation=<rot 90 0 0>}]
 deform_tail_r [{shapeType=Capsule, length=1, radius=1, shapeName=deform_tail_r_capsule, rotation=<rot 90 0 0>}]
 dyn_hair_1_f [{shapeType=Capsule, length=4.1757, radius=3, shapeName=dyn_hair_1_f_capsule, center=<3.0625 0 0>, rotation=<rot 90 0 0>}]
 dyn_hair_1_l [{shapeType=Capsule, length=5, radius=3, shapeName=dyn_hair_1_l_capsule, center=<3.4362 -1.4964 0>, rotation=<rot 90 0 0>}]
 dyn_hair_1_r [{shapeType=Capsule, length=5, radius=3, shapeName=dyn_hair_1_r_capsule, center=<3.1265 -1.411 0>, rotation=<rot 90 0 0>}]
 dyn_hair_2_l [{shapeType=Capsule, length=3.2091, radius=3, shapeName=dyn_hair_2_l_capsule, center=<2.2391 -1.4964 0>, rotation=<rot 90 0 0>}]
 dyn_hair_2_r [{shapeType=Capsule, length=2.7506, radius=3, shapeName=dyn_hair_2_r_capsule, center=<2.2071 -1.411 0>, rotation=<rot 90 0 0>}]
 dyn_hair_3_l [{shapeType=Capsule, length=5, radius=3, shapeName=dyn_hair_3_l_capsule, center=<3.4362 -1.4964 0>, rotation=<rot 90 0 0>}]
 dyn_hair_3_r [{shapeType=Capsule, length=5, radius=3, shapeName=dyn_hair_3_r_capsule, center=<3.1265 -1.411 0>, rotation=<rot 90 0 0>}]
 dyn_hair_f [{shapeType=Capsule, length=4.1757, radius=3, shapeName=dyn_hair_f_capsule, center=<3.0625 0 0>, rotation=<rot 90 0 0>}]
 dyn_hair_l [{shapeType=Capsule, length=5, radius=3, shapeName=dyn_hair_l_capsule, center=<3.4362 -1.4964 0>, rotation=<rot 90 0 0>}]
 dyn_hair_r [{shapeType=Capsule, length=5, radius=3, shapeName=dyn_hair_r_capsule, center=<3.1265 -1.411 0>, rotation=<rot 90 0 0>}]
 dyn_tail_1_l [{shapeType=Capsule, length=3.6871, radius=4.3129, shapeName=dyn_tail_1_l_capsule, center=<3.3683 0 0>, rotation=<rot 90 0 0>}]
 dyn_tail_1_r [{shapeType=Capsule, length=5, radius=5, shapeName=dyn_tail_1_r_capsule, center=<4.1089 0 0>, rotation=<rot 90 0 0>}]
 dyn_tail_2_l [{shapeType=Capsule, length=3.6871, radius=4.3129, shapeName=dyn_tail_2_l_capsule, center=<3.3683 0 0>, rotation=<rot 90 0 0>}]
 dyn_tail_2_r [{shapeType=Capsule, length=5, radius=5, shapeName=dyn_tail_2_r_capsule, center=<4.1089 0 0>, rotation=<rot 90 0 0>}]
 dyn_tail_3_l [{shapeType=Capsule, length=1.9327, radius=3.6952, shapeName=dyn_tail_3_l_capsule, center=<1.8675 0 0>, rotation=<rot 90 0 0>}]
 dyn_tail_3_r [{shapeType=Capsule, length=3.3694, radius=3.9484, shapeName=dyn_tail_3_r_capsule, center=<2.7915 0 0>, rotation=<rot 90 0 0>}]
 dyn_tail_l [{shapeType=Capsule, length=5, radius=3, shapeName=dyn_tail_l_capsule, center=<4.6271 0 0>, rotation=<rot 90 0 0>}]
 dyn_tail_r [{shapeType=Capsule, length=5, radius=5, shapeName=dyn_tail_r_capsule, center=<4.1089 0 0>, rotation=<rot 90 0 0>}]
 head [
  {shapeType=Sphere, radius=9.2696, shapeName=head_sphere, center=<6.5363 -1.4069 0>}
  {shapeType=Capsule, length=12.6269, radius=6.7153, shapeName=head_capsule, center=<4.5787 0.1953 0>, rotation=<rot 90 -165.9638 -126.289>}
  {shapeType=Capsule, length=17.2898, radius=1.9277, shapeName=head_capsule, center=<2.427 -1.7252 0>}
 ]
 neck_01 [{shapeType=Capsule, length=5, radius=5, shapeName=neck_01_capsule, rotation=<rot 90 0 0>}]
 neck_02 [{shapeType=Capsule, length=3.8891, radius=6.1109, shapeName=neck_02_capsule, rotation=<rot 90 0 0>}]
 spine_03 [{shapeType=Capsule, length=6.2637, radius=7.8229, shapeName=spine_03_capsule, center=<0 1.258 0>, rotation=<rot 0 -90 0>}]
 spine_04 [{shapeType=Capsule, length=10.0678, radius=7.8229, shapeName=spine_04_capsule, center=<0 0.1376 0>, rotation=<rot 0 -90 0>}]
 spine_05 [
  {shapeType=Capsule, length=15.0039, radius=8.9353, shapeName=spine_05_capsule}
  {shapeType=Capsule, length=9.0968, radius=14.0071, shapeName=spine_05_capsule, center=<-5.4517 -14.6033 0>, rotation=<rot 90 0 0>}
 ]
 upperarm_l [{shapeType=Capsule, length=0.505, radius=0.505, shapeName=upperarm_l_capsule, rotation=<rot 0 -90 0>}]
 upperarm_r [{shapeType=Capsule, length=0.505, radius=0.505, shapeName=upperarm_r_capsule, rotation=<rot 0 -90 0>}]
}
constraints [
 {
  bone1Name=neck_01
  bone2Name=spine_05
  swing1LimitDegrees=45
  swing1Motion=Free
  swing2LimitDegrees=45
  swing2Motion=Free
  twistLimitDegrees=45
  twistMotion=Free
 }
 {
  bone1Name=neck_02
  bone2Name=neck_01
  swing1LimitDegrees=45
  swing1Motion=Free
  swing2LimitDegrees=45
  swing2Motion=Free
  twistLimitDegrees=45
  twistMotion=Free
 }
 {
  bone1Name=head
  bone2Name=neck_02
  swing1LimitDegrees=45
  swing1Motion=Free
  swing2LimitDegrees=45
  swing2Motion=Free
  twistLimitDegrees=45
  twistMotion=Free
 }
 {
  bone1Name=clavicle_l
  bone2Name=spine_05
  swing1LimitDegrees=45
  swing1Motion=Free
  swing2LimitDegrees=45
  swing2Motion=Free
  twistLimitDegrees=45
  twistMotion=Free
 }
 {
  bone1Name=deform_tail_l
  bone2Name=head
  swing1LimitDegrees=45
  swing1Motion=Free
  swing2LimitDegrees=45
  swing2Motion=Free
  twistLimitDegrees=45
  twistMotion=Free
 }
 {bone1Name=dyn_tail_l, bone2Name=deform_tail_l, swing1LimitDegrees=2, swing2LimitDegrees=2, twistLimitDegrees=5}
 {bone1Name=dyn_tail_1_l, bone2Name=dyn_tail_l, swing1LimitDegrees=35, swing2LimitDegrees=35, twistLimitDegrees=5}
 {bone1Name=dyn_tail_2_l, bone2Name=dyn_tail_1_l, swing1LimitDegrees=35, swing2LimitDegrees=35, twistLimitDegrees=5}
 {bone1Name=dyn_tail_3_l, bone2Name=dyn_tail_2_l, swing1LimitDegrees=35, swing2LimitDegrees=35, twistLimitDegrees=5}
 {
  bone1Name=deform_tail_r
  bone2Name=head
  swing1LimitDegrees=45
  swing1Motion=Free
  swing2LimitDegrees=45
  swing2Motion=Free
  twistLimitDegrees=45
  twistMotion=Free
 }
 {bone1Name=dyn_tail_r, bone2Name=deform_tail_r, swing1LimitDegrees=2, swing2LimitDegrees=2, twistLimitDegrees=5}
 {bone1Name=dyn_tail_1_r, bone2Name=dyn_tail_r, swing1LimitDegrees=35, swing2LimitDegrees=35, twistLimitDegrees=5}
 {bone1Name=dyn_tail_2_r, bone2Name=dyn_tail_1_r, swing1LimitDegrees=35, swing2LimitDegrees=35, twistLimitDegrees=5}
 {bone1Name=dyn_tail_3_r, bone2Name=dyn_tail_2_r, swing1LimitDegrees=35, swing2LimitDegrees=35, twistLimitDegrees=5}
 {
  bone1Name=clavicle_r
  bone2Name=spine_05
  swing1LimitDegrees=45
  swing1Motion=Free
  swing2LimitDegrees=45
  swing2Motion=Free
  twistLimitDegrees=45
  twistMotion=Free
 }
 {
  bone1Name=deform_hair_f
  bone2Name=head
  swing1LimitDegrees=45
  swing1Motion=Free
  swing2LimitDegrees=45
  swing2Motion=Free
  twistLimitDegrees=45
  twistMotion=Free
 }
 {bone1Name=dyn_hair_f, bone2Name=deform_hair_f, swing1LimitDegrees=25, swing2LimitDegrees=5, twistLimitDegrees=1}
 {bone1Name=dyn_hair_1_f, bone2Name=dyn_hair_f, swing1LimitDegrees=25, swing2LimitDegrees=25, twistLimitDegrees=1}
 {
  bone1Name=deform_hair_l
  bone2Name=head
  swing1LimitDegrees=45
  swing1Motion=Free
  swing2LimitDegrees=45
  swing2Motion=Free
  twistLimitDegrees=45
  twistMotion=Free
 }
 {bone1Name=dyn_hair_l, bone2Name=deform_hair_l, swing1LimitDegrees=10, swing2LimitDegrees=10, twistLimitDegrees=1}
 {bone1Name=dyn_hair_1_l, bone2Name=dyn_hair_l, swing1LimitDegrees=25, swing2LimitDegrees=25, twistLimitDegrees=1}
 {bone1Name=dyn_hair_2_l, bone2Name=dyn_hair_1_l, swing1LimitDegrees=25, swing2LimitDegrees=25, twistLimitDegrees=1}
 {bone1Name=dyn_hair_3_l, bone2Name=dyn_hair_l, swing1LimitDegrees=25, swing2LimitDegrees=25, twistLimitDegrees=1}
 {
  bone1Name=deform_hair_r
  bone2Name=head
  swing1LimitDegrees=45
  swing1Motion=Free
  swing2LimitDegrees=45
  swing2Motion=Free
  twistLimitDegrees=45
  twistMotion=Free
 }
 {bone1Name=dyn_hair_r, bone2Name=deform_hair_r, swing1LimitDegrees=10, swing2LimitDegrees=10, twistLimitDegrees=1}
 {bone1Name=dyn_hair_1_r, bone2Name=dyn_hair_r, swing1LimitDegrees=25, swing2LimitDegrees=25, twistLimitDegrees=1}
 {bone1Name=dyn_hair_2_r, bone2Name=dyn_hair_1_r, swing1LimitDegrees=25, swing2LimitDegrees=25, twistLimitDegrees=1}
 {bone1Name=dyn_hair_3_r, bone2Name=dyn_hair_r, swing1LimitDegrees=25, swing2LimitDegrees=25, twistLimitDegrees=1}
 {
  bone1Name=spine_05
  bone2Name=spine_04
  swing1LimitDegrees=45
  swing1Motion=Free
  swing2LimitDegrees=45
  swing2Motion=Free
  twistLimitDegrees=45
  twistMotion=Free
 }
 {
  bone1Name=spine_04
  bone2Name=spine_03
  swing1LimitDegrees=45
  swing1Motion=Free
  swing2LimitDegrees=45
  swing2Motion=Free
  twistLimitDegrees=45
  twistMotion=Free
 }
 {
  bone1Name=upperarm_l
  bone2Name=clavicle_l
  swing1LimitDegrees=45
  swing1Motion=Free
  swing2LimitDegrees=45
  swing2Motion=Free
  twistLimitDegrees=45
  twistMotion=Free
 }
 {
  bone1Name=upperarm_r
  bone2Name=clavicle_r
  swing1LimitDegrees=45
  swing1Motion=Free
  swing2LimitDegrees=45
  swing2Motion=Free
  twistLimitDegrees=45
  twistMotion=Free
 }
]
